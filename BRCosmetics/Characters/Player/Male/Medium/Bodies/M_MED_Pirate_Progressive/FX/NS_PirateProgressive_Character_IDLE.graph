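# Niagara idle effect for the progressive pirate character, layering ash, smoke, glow sprites and lights on skeletal mesh bones, tinted by User.Color.
# Ash and rising smoke spawn continuously at partition-selected bones, with ash following character movement less over its lifetime.
# Burst-spawned glows and lights persist on chosen bones, pulsing on index-staggered System.Age curves.
# System-level FadeIn scales all layers, and glow partitions set a renderer VisibilityTag.
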NiagaraSystem /BRCosmetics/Characters/Player/Male/Medium/Bodies/M_MED_Pirate_Progressive/FX/NS_PirateProgressive_Character_IDLE
dependencies {
 usedDataInterfaces [
  &NiagaraDataInterfaceCurve
  &NiagaraDataInterfaceArrayInt32
  &NiagaraDataInterfaceVector2DCurve
  &NiagaraDataInterfaceArrayFloat3
  &NiagaraDataInterfaceArrayFloat
  &NiagaraDataInterfaceSkeletalMesh
 ]
 usedDynamicInputs []
 usedModules [
  &self:SystemSpawnScript.SystemScriptSource.SystemScriptGraph.NiagaraNodeAssignment_0.SetVariables_#1
  &self:SystemSpawnScript.SystemScriptSource.SystemScriptGraph.NiagaraNodeAssignment_1.SetVariables_#2
  &self:SystemSpawnScript.SystemScriptSource.SystemScriptGraph.NiagaraNodeAssignment_2.SetVariables_#3
  &SystemState
 ]
}
emitters {
 Ash {
  emitterUpdateScript [
   &EmitterState(Life Cycle Mode=::NewEnumerator1, UseLoopDelay=true)
   &SpawnRate(SpawnRate=30, Use Spawn Probability=false)
  ]
  particleSpawnScript [
   &PartitionParticles(Number Of Partitions=16)
   /SetVariables(Particles.SpriteRotationRate=&RandomRangeFloat(Minimum=-0.7, Maximum=0.7, Evaluation Type=::NewEnumerator1)) {isSetParameters=true}
   &InitializeParticle(
    Color=$User.Color
    Lifetime=&RandomRangeFloat(Minimum=1.4, Maximum=2, Evaluation Type=::NewEnumerator1)
    Sprite Rotation Angle=&RandomRangeFloat(Minimum=0, Maximum=1, Evaluation Type=::NewEnumerator1)
    Sprite Rotation Mode=::NewEnumerator1
    Sprite Size=&Vector2DFromFloat(Value=&RandomRangeFloat(Minimum=1, Maximum=4, Evaluation Type=::NewEnumerator1))
    Sprite Size Mode=::NewEnumerator4
    UsePositionOffset=false
    Write Material Random=true
   )
   &Color(Color=$User.Color)
   &SkeletalMeshLocation(
    Bone / Socket Index=$Particles.PartitionParticles.Partition
    Bone Sampling Mode=::NewEnumerator4
    MaskSpawnGroup=false
    Sampled Position Offset=&SelectVectorFromArray(Direct Array Index=$Particles.PartitionParticles.Partition, Vector Selection Array=$System.SkelOffsetArray)
    Skeletal Mesh=$System.Skeletal Mesh
   )
   &InheritVelocity(Inherited Velocity Amount Scale=<0.25 0.25 0.25>, Inherited Velocity Speed Limit=0, Speed Limit=false, UseSpeedThreshold=false)
   &CameraOffset(Camera Offset Amount=&RandomRangeFloat(Minimum=-2, Maximum=2, Evaluation Type=::NewEnumerator1))
   &DynamicMaterialParameters(
    Index 0 Param 1=&RandomRangeFloat(Minimum=0, Maximum=1, Evaluation Type=::NewEnumerator1)
    Index 0 Param 2=&RandomRangeFloat(Minimum=0.1, Maximum=1, Evaluation Type=::NewEnumerator1)
    Module.Index0Param0Write=true
    Module.Index0Param1Write=true
    Module.Index0Param2Write=false
    Module.Index0Param3Write=false
   )
  ]
  particleUpdateScript [
   &ParticleState()
   &SpriteRotationRate(Rotation Rate=&AngleConversion(
     Angle:NiagaraFloat=$Particles.SpriteRotationRate
     Angle Input:ENiagara_AngleInput=::NewEnumerator1
     Angle Output:ENiagara_AngleInput=::NewEnumerator0
    ))
   &ScaleSpriteSize(
    Scale Sprite Size Mode=::NewEnumerator0
    Uniform Scale Factor=&FloatFromCurve(
     FloatCurve=DataInterface() {propertyValues {Curve {keys [<at 0 1 RCIM_Cubic>, {arriveTangent=-0.5, interpMode=RCIM_Cubic, leaveTangent=-0.5, time=0.7, value=0.8}, <at 1 0.5>]}}}
     Scale Curve=$System.FadeIn
    )
   )
   &ScaleColor(Scale Alpha=$System.FadeIn, ScaleA=true, ScaleRGB=true)
   &Drag(Drag=0.5, Use Linear Drag=true, Use Rotational Drag=true)
   &InheritSourceMovement(
    Applied Position Scale=&VectorFromFloat(Value=&FloatFromCurve(FloatCurve=DataInterface() {propertyValues {Curve {keys [<at 0 1>,<at 1 0>]}}}))
    ApplyForce=false
    ApplyPosition=true
    ApplyPositionFalloff=false
   )
   &SolveForcesAndVelocity(Clamp Velocity=false, Limit Acceleration=false)
   /Niagara/Modules/Update/SubUV/SubUVAnimation(Number Of Frames=40)
   &DynamicMaterialParameters(
    Index 0 Param 3=1
    Index 0 Param 4=1
    Module.Index0Param0Write=false
    Module.Index0Param1Write=false
    Module.Index0Param2Write=true
    Module.Index0Param3Write=true
   ) {module=DynamicMaterialParameters001}
  ]
  rendererClasses [&NiagaraSpriteRendererProperties]
  renderers [{rendererClass=&NiagaraSpriteRendererProperties}]
 }
 Glows {
  emitterUpdateScript [
   &EmitterState(
    Inactive Response=::NewEnumerator1
    Life Cycle Mode=::NewEnumerator1
    Loop Behavior=::NewEnumerator1
    Loop Duration Mode=::NewEnumerator1
    UseLoopDelay=true
   )
   &SpawnBurst_Instantaneous(Spawn Count=3, Use Spawn Probability=false, UseLoopCountLimit=false)
  ]
  particleSpawnScript [
   &PartitionParticles(ID to Partition:NiagaraInt32=&ReturnExecIndex(), Number Of Partitions=3)
   /SetVariables(Particles.VisibilityTag=$Particles.PartitionParticles.Partition) {isSetParameters=true}
   &InitializeParticle(
    Color=$User.Color
    Lifetime=1.4
    Sprite Size Mode=::NewEnumerator3
    Uniform Sprite Size=&SelectFloatFromArray(Direct Array Index=$Particles.PartitionParticles.Partition, Float Selection Array=DataInterface() {propertyValues {FloatData [60,4.4,7]}})
    UsePositionOffset=false
    Write Material Random=true
   )
   /SetVariables(Particles.BoneIndex:NiagaraInt32=&SelectIntFromArray(Direct Array Index=$Particles.PartitionParticles.Partition, Int Selection Array=DataInterface() {propertyValues {IntData [3,18,18]}})) {isSetParameters=true}
   &CameraOffset(Camera Offset Amount=&SelectFloatFromArray(Direct Array Index=$Particles.PartitionParticles.Partition, Float Selection Array=DataInterface() {propertyValues {FloatData [0,5,0]}}))
  ]
  particleUpdateScript [
   &ParticleState(Kill Particles When Lifetime Has Elapsed=false)
   &Color(
    Color=$User.Color
    Scale Alpha=&Multiply_Float(
     A=&Lerp_Float(
      A=0.75
      B=1
      Alpha=&FloatFromCurve(
       FloatCurve=DataInterface() {propertyValues {Curve {keys [<at 0 0 RCIM_Cubic>,<at 0.5 1 RCIM_Cubic>,<at 1 0>]}}}
       CurveIndex=&ModuloFloat(
        A=&Add_Float(
         A=&Multiply_Float(A=$System.Age, B=0.3)
         B=&ReturnNormalizedExecIndex(Normalized Index Scale=0.25)
        )
        B=1
       )
      )
     )
     B=$System.FadeIn
    )
    Scale Color=&Multiply_VectorByFloat(Vector=<1 1 1>, Float=$System.FadeIn)
   )
   &SkeletalMeshLocation(
    Bone / Socket Index=$Particles.BoneIndex
    Bone Sampling Mode=::NewEnumerator4
    MaskSpawnGroup=false
    Sampled Position Offset=&SelectVectorFromArray(
     Direct Array Index=$Particles.PartitionParticles.Partition
     Vector Selection Array=DataInterface() {propertyValues {FloatData [<10 2 0>, <4.7 7.5 -2.6>, <-1 6 0>]}}
    )
    Skeletal Mesh=$System.Skeletal Mesh
   )
   &ScaleColor(
    Scale Alpha=&Multiply_Float(
     A=&Lerp_Float(
      A=0.75
      B=1
      Alpha=&FloatFromCurve(
       FloatCurve=DataInterface() {propertyValues {Curve {keys [<at 0 0 RCIM_Cubic>,<at 0.5 1 RCIM_Cubic>,<at 1 0>]}}}
       CurveIndex=&ModuloFloat(
        A=&Add_Float(
         A=&Multiply_Float(A=$System.Age, B=0.3)
         B=&ReturnNormalizedExecIndex(Normalized Index Scale=0.25)
        )
        B=1
       )
      )
     )
     B=$System.FadeIn
    )
    Scale RGB=&Multiply_VectorByFloat(Vector=<1 1 1>, Float=$System.FadeIn)
    ScaleA=true
    ScaleRGB=true
   ) {enabled=false}
  ]
  rendererClasses [&NiagaraSpriteRendererProperties]
  renderers [
   {rendererClass=&NiagaraSpriteRendererProperties}
   {rendererClass=&NiagaraSpriteRendererProperties, rendererIndex=1}
   {rendererClass=&NiagaraSpriteRendererProperties, rendererIndex=2}
  ]
 }
 Lights {
  emitterUpdateScript [
   &EmitterState(
    Inactive Response=::NewEnumerator1
    Life Cycle Mode=::NewEnumerator1
    Loop Behavior=::NewEnumerator1
    Loop Duration Mode=::NewEnumerator1
    UseLoopDelay=true
   )
   &SpawnBurst_Instantaneous(Spawn Count=4, Use Spawn Probability=false, UseLoopCountLimit=false)
  ]
  particleSpawnScript [
   &PartitionParticles(ID to Partition:NiagaraInt32=&ReturnExecIndex(), Number Of Partitions=4)
   /SetVariables(Particles.BoneIndex:NiagaraInt32=&SelectIntFromArray(Direct Array Index=$Particles.PartitionParticles.Partition, Int Selection Array=DataInterface() {propertyValues {IntData [3,9,12,18]}})) {isSetParameters=true}
   &InitializeParticle(Color=$User.Color, Lifetime=1.4, Sprite Size=<33 33>, Sprite Size Mode=::NewEnumerator4, UsePositionOffset=false, Write Material Random=true)
   &CameraOffset(Camera Offset Amount=5)
  ]
  particleUpdateScript [
   &ParticleState(Kill Particles When Lifetime Has Elapsed=false)
   /SetVariables(Particles.LightExponent=6) {isSetParameters=true}
   &Color(Color=$User.Color)
   &CascadeConversion_LightProperties(
    LightBrightnessScale:NiagaraFloat=&Lerp_Float(
     A=15
     B=20
     Alpha=&FloatFromCurve(
      FloatCurve=DataInterface() {propertyValues {Curve {keys [<at 0 0 RCIM_Cubic>,<at 0.5 1 RCIM_Cubic>,<at 1 0>]}}}
      CurveIndex=&ModuloFloat(
       A=&Add_Float(
        A=&Multiply_Float(A=$System.Age, B=0.3)
        B=&ReturnNormalizedExecIndex(Normalized Index Scale=0.25)
       )
       B=1
      )
     )
    )
    LightColorScale=<1 1 1>
    LightRadiusScale=1
   )
   &SkeletalMeshLocation(
    Bone / Socket Index=$Particles.BoneIndex
    Bone Sampling Mode=::NewEnumerator4
    MaskSpawnGroup=false
    Sampled Position Offset=&SelectVectorFromArray(
     Direct Array Index=$Particles.PartitionParticles.Partition
     Vector Selection Array=DataInterface() {propertyValues {FloatData [<3 6 0>, <5 0 0>, <-8.1 2 2>, <3 5 0>]}}
    )
    Skeletal Mesh=$System.Skeletal Mesh
   )
   &FN_Light_Attributes(
    Color=$Particles.Color
    Create Color Attribute=true
    Create Exponent Attribute=true
    Create Radius Attribute=true
    Exponent=2
    Radius=&Multiply_Float(
     A=&Lerp_Float(
      A=0.75
      B=1
      Alpha=&FloatFromCurve(
       FloatCurve=DataInterface() {propertyValues {Curve {keys [<at 0 0 RCIM_Cubic>,<at 0.5 1 RCIM_Cubic>,<at 1 0>]}}}
       CurveIndex=&ModuloFloat(
        A=&Add_Float(
         A=&Multiply_Float(A=$System.Age, B=0.3)
         B=&ReturnNormalizedExecIndex(Normalized Index Scale=0.25)
        )
        B=1
       )
      )
     )
     B=&SelectFloatFromArray(Direct Array Index=$Particles.PartitionParticles.Partition, Float Selection Array=DataInterface() {propertyValues {FloatData [20,10,20,12]}})
    )
    Scale Color=&Multiply_Float(A=3, B=$System.FadeIn)
   )
  ]
  rendererClasses [&NiagaraLightRendererProperties]
  renderers [{rendererClass=&NiagaraLightRendererProperties, rendererIndex=0}]
 }
 SmokeLoopIdle {
  emitterUpdateScript [
   &EmitterState(Life Cycle Mode=::NewEnumerator1, Loop Duration=1, UseLoopDelay=true)
   &SpawnRate(SpawnRate=7, Use Spawn Probability=false)
  ]
  particleSpawnScript [
   &PartitionParticles(Number Of Partitions=9)
   /SetVariables(Particles.BoneIndex:NiagaraInt32=&SelectIntFromArray(
     Direct Array Index=$Particles.PartitionParticles.Partition
     Int Selection Array=DataInterface() {propertyValues {IntData [3,5,9,11,12,14,15,17]}}
    )) {isSetParameters=true}
   &InitializeParticle(
    Color=$User.Color
    Lifetime=&RandomRangeFloat(Minimum=1.15, Maximum=1.6, Evaluation Type=::NewEnumerator1)
    Sprite Rotation Angle=-0.25
    Sprite Rotation Mode=::NewEnumerator1
    Sprite Size=&RandomRangeVector2D(Minimum=<25 30>, Maximum=<35 40>, Evaluation Type=::NewEnumerator1)
    Sprite Size Mode=::NewEnumerator4
    UsePositionOffset=false
    Write Material Random=true
   )
   &Color(Color=$User.Color)
   &CameraOffset(Camera Offset Amount=&RandomRangeFloat(Minimum=8, Maximum=20, Evaluation Type=::NewEnumerator1))
   &DynamicMaterialParameters(
    Index 0 Param 1=&RandomRangeFloat(Minimum=0, Maximum=1, Evaluation Type=::NewEnumerator1)
    Index 0 Param 2=&RandomRangeFloat(Minimum=0, Maximum=1, Evaluation Type=::NewEnumerator1)
    Module.Index0Param0Write=true
    Module.Index0Param1Write=true
    Module.Index0Param2Write=false
    Module.Index0Param3Write=false
   )
   &AddVelocity(Velocity=<0 0 20>)
  ]
  particleUpdateScript [
   &ParticleState()
   &ScaleSpriteSize(
    Scale Factor=&Vector2DFromCurve(
     Vector2Curve=DataInterface() {propertyValues {XCurve {keys [<at 0 0.8 RCIM_Cubic>,<at 1 1.4>]}, YCurve {keys [<at 0 1 RCIM_Cubic>,<at 1 0.8>]}}}
     ScaleCurve=&Multiply_Vector2D_ByFloat(Float=$System.FadeIn, Vector2D=<1 1>)
    )
    Scale Sprite Size Mode=::NewEnumerator2
   )
   &ScaleColor(
    Scale Alpha=&FloatFromCurve(FloatCurve=DataInterface() {propertyValues {Curve {keys [<at 0 0>,<at 0.45 0.3>,<at 1 0>]}}}, Scale Curve=$System.FadeIn)
    ScaleA=true
    ScaleRGB=true
   )
   &SkeletalMeshLocation(
    Bone / Socket Index=$Particles.BoneIndex
    Bone Sampling Mode=::NewEnumerator4
    MaskSpawnGroup=false
    Sampled Position Offset=&SelectVectorFromArray(Direct Array Index=$Particles.BoneIndex, Vector Selection Array=$System.SkelOffsetArray)
    Skeletal Mesh=$System.Skeletal Mesh
   )
   &DynamicMaterialParameters(
    Index 0 Param 3=&FloatFromCurve(FloatCurve=DataInterface() {propertyValues {Curve {keys [<at 0 0.5 RCIM_Cubic>, {arriveTangent=0.5, interpMode=RCIM_Cubic, leaveTangent=0.5, time=0.4, value=0}, <at 1 1 RCIM_Cubic>]}}})
    Module.Index0Param0Write=false
    Module.Index0Param1Write=false
    Module.Index0Param2Write=true
    Module.Index0Param3Write=true
   ) {module=DynamicMaterialParameters001}
   /SetVariables(Particles.SpriteAlignment=<0 0 1>) {isSetParameters=true}
   &SolveForcesAndVelocity(Clamp Velocity=false, Limit Acceleration=false)
  ]
  rendererClasses [&NiagaraSpriteRendererProperties]
  renderers [{rendererClass=&NiagaraSpriteRendererProperties}]
 }
}
summary {
 systemName=NS_PirateProgressive_Character_IDLE
 emitters [
  {emitterName=Ash, rendererClasses [&NiagaraSpriteRendererProperties]}
  {emitterName=SmokeLoopIdle, rendererClasses [&NiagaraSpriteRendererProperties]}
  {emitterName=Lights, rendererClasses [&NiagaraLightRendererProperties]}
  {emitterName=Glows, rendererClasses [&NiagaraSpriteRendererProperties]}
 ]
 userVariables [User.Color {type=LinearColor, description="", defaultValue {struct=LinearColor, value=<rgba 0.0872 0.6376 0.3272 1>}}]
}
