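# Physics asset for the tier 1 skirt of the medium male pirate player body, defining its collision bodies and joints.
# Eleven four-bone chains of identical thin boxes form the skirt, with capsules on the pelvis and both thighs.
# Five chains per side hang from that side's thigh, while the center back chain hangs from the pelvis.
# All joints, including thigh to pelvis, share the same swing and tight twist limits.
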PhysicsAsset /Game/Characters/Player/Male/Medium/Bodies/M_MED_Pirate_Progressive/Meshes/M_MED_Pirate_T1_Skirt_Physics
bodies {
 dyn_skirt_back_T1_01 [{shapeType=Box, extentX=10, extentY=2, extentZ=10, center=<5 0 0>}]
 dyn_skirt_back_T1_01_l [{shapeType=Box, extentX=10, extentY=2, extentZ=10, center=<5 0 0>}]
 dyn_skirt_back_T1_01_r [{shapeType=Box, extentX=10, extentY=2, extentZ=10, center=<5 0 0>}]
 dyn_skirt_back_T1_02 [{shapeType=Box, extentX=10, extentY=2, extentZ=10, center=<5 0 0>}]
 dyn_skirt_back_T1_02_l [{shapeType=Box, extentX=10, extentY=2, extentZ=10, center=<5 0 0>}]
 dyn_skirt_back_T1_02_r [{shapeType=Box, extentX=10, extentY=2, extentZ=10, center=<5 0 0>}]
 dyn_skirt_back_T1_03 [{shapeType=Box, extentX=10, extentY=2, extentZ=10, center=<5 0 0>}]
 dyn_skirt_back_T1_03_l [{shapeType=Box, extentX=10, extentY=2, extentZ=10, center=<5 0 0>}]
 dyn_skirt_back_T1_03_r [{shapeType=Box, extentX=10, extentY=2, extentZ=10, center=<5 0 0>}]
 dyn_skirt_back_T1_04 [{shapeType=Box, extentX=10, extentY=2, extentZ=10, center=<5 0 0>}]
 dyn_skirt_back_T1_04_l [{shapeType=Box, extentX=10, extentY=2, extentZ=10, center=<5 0 0>}]
 dyn_skirt_back_T1_04_r [{shapeType=Box, extentX=10, extentY=2, extentZ=10, center=<5 0 0>}]
 dyn_skirt_back_side_T1_01_l [{shapeType=Box, extentX=10, extentY=2, extentZ=10, center=<5 0 0>}]
 dyn_skirt_back_side_T1_01_r [{shapeType=Box, extentX=10, extentY=2, extentZ=10, center=<5 0 0>}]
 dyn_skirt_back_side_T1_02_l [{shapeType=Box, extentX=10, extentY=2, extentZ=10, center=<5 0 0>}]
 dyn_skirt_back_side_T1_02_r [{shapeType=Box, extentX=10, extentY=2, extentZ=10, center=<5 0 0>}]
 dyn_skirt_back_side_T1_03_l [{shapeType=Box, extentX=10, extentY=2, extentZ=10, center=<5 0 0>}]
 dyn_skirt_back_side_T1_03_r [{shapeType=Box, extentX=10, extentY=2, extentZ=10, center=<5 0 0>}]
 dyn_skirt_back_side_T1_04_l [{shapeType=Box, extentX=10, extentY=2, extentZ=10, center=<5 0 0>}]
 dyn_skirt_back_side_T1_04_r [{shapeType=Box, extentX=10, extentY=2, extentZ=10, center=<5 0 0>}]
 dyn_skirt_front_T1_01_l [{shapeType=Box, extentX=10, extentY=2, extentZ=10, center=<5 0 0>}]
 dyn_skirt_front_T1_01_r [{shapeType=Box, extentX=10, extentY=2, extentZ=10, center=<5 0 0>}]
 dyn_skirt_front_T1_02_l [{shapeType=Box, extentX=10, extentY=2, extentZ=10, center=<5 0 0>}]
 dyn_skirt_front_T1_02_r [{shapeType=Box, extentX=10, extentY=2, extentZ=10, center=<5 0 0>}]
 dyn_skirt_front_T1_03_l [{shapeType=Box, extentX=10, extentY=2, extentZ=10, center=<5 0 0>}]
 dyn_skirt_front_T1_03_r [{shapeType=Box, extentX=10, extentY=2, extentZ=10, center=<5 0 0>}]
 dyn_skirt_front_T1_04_l [{shapeType=Box, extentX=10, extentY=2, extentZ=10, center=<5 0 0>}]
 dyn_skirt_front_T1_04_r [{shapeType=Box, extentX=10, extentY=2, extentZ=10, center=<5 0 0>}]
 dyn_skirt_front_mid_T1_01_l [{shapeType=Box, extentX=10, extentY=2, extentZ=10, center=<5 0 0>}]
 dyn_skirt_front_mid_T1_01_r [{shapeType=Box, extentX=10, extentY=2, extentZ=10, center=<5 0 0>}]
 dyn_skirt_front_mid_T1_02_l [{shapeType=Box, extentX=10, extentY=2, extentZ=10, center=<5 0 0>}]
 dyn_skirt_front_mid_T1_02_r [{shapeType=Box, extentX=10, extentY=2, extentZ=10, center=<5 0 0>}]
 dyn_skirt_front_mid_T1_03_l [{shapeType=Box, extentX=10, extentY=2, extentZ=10, center=<5 0 0>}]
 dyn_skirt_front_mid_T1_03_r [{shapeType=Box, extentX=10, extentY=2, extentZ=10, center=<5 0 0>}]
 dyn_skirt_front_mid_T1_04_l [{shapeType=Box, extentX=10, extentY=2, extentZ=10, center=<5 0 0>}]
 dyn_skirt_front_mid_T1_04_r [{shapeType=Box, extentX=10, extentY=2, extentZ=10, center=<5 0 0>}]
 dyn_skirt_side_T1_01_l [{shapeType=Box, extentX=10, extentY=2, extentZ=10, center=<5 0 0>}]
 dyn_skirt_side_T1_01_r [{shapeType=Box, extentX=10, extentY=2, extentZ=10, center=<5 0 0>}]
 dyn_skirt_side_T1_02_l [{shapeType=Box, extentX=10, extentY=2, extentZ=10, center=<5 0 0>}]
 dyn_skirt_side_T1_02_r [{shapeType=Box, extentX=10, extentY=2, extentZ=10, center=<5 0 0>}]
 dyn_skirt_side_T1_03_l [{shapeType=Box, extentX=10, extentY=2, extentZ=10, center=<5 0 0>}]
 dyn_skirt_side_T1_03_r [{shapeType=Box, extentX=10, extentY=2, extentZ=10, center=<5 0 0>}]
 dyn_skirt_side_T1_04_l [{shapeType=Box, extentX=10, extentY=2, extentZ=10, center=<5 0 0>}]
 dyn_skirt_side_T1_04_r [{shapeType=Box, extentX=10, extentY=2, extentZ=10, center=<5 0 0>}]
 pelvis [{shapeType=Capsule, length=11, radius=9.5, center=<-11.9354 0.0001 0>}]
 thigh_l [
  {shapeType=Capsule, length=25, radius=9, center=<-25 0 0>, rotation=<rot 90 0 0>}
  {shapeType=Capsule, length=17, radius=9, center=<-25.041 4.4269 0>, rotation=<rot 90 0 0>}
 ]
 thigh_r [
  {shapeType=Capsule, length=25, radius=9, center=<25 0 0>, rotation=<rot 90 0 -179.9999>}
  {shapeType=Capsule, length=17, radius=9, center=<25.041 -4.4269 0>, rotation=<rot 90 0 180>}
 ]
}
constraints [
 {bone1Name=dyn_skirt_front_mid_T1_01_r, bone2Name=thigh_r, swing1LimitDegrees=45, swing2LimitDegrees=45, twistLimitDegrees=5}
 {
  bone1Name=dyn_skirt_front_mid_T1_02_r
  bone2Name=dyn_skirt_front_mid_T1_01_r
  swing1LimitDegrees=45
  swing2LimitDegrees=45
  twistLimitDegrees=5
 }
 {
  bone1Name=dyn_skirt_front_mid_T1_03_r
  bone2Name=dyn_skirt_front_mid_T1_02_r
  swing1LimitDegrees=45
  swing2LimitDegrees=45
  twistLimitDegrees=5
 }
 {
  bone1Name=dyn_skirt_front_mid_T1_04_r
  bone2Name=dyn_skirt_front_mid_T1_03_r
  swing1LimitDegrees=45
  swing2LimitDegrees=45
  twistLimitDegrees=5
 }
 {bone1Name=dyn_skirt_front_T1_01_r, bone2Name=thigh_r, swing1LimitDegrees=45, swing2LimitDegrees=45, twistLimitDegrees=5}
 {bone1Name=dyn_skirt_front_T1_02_r, bone2Name=dyn_skirt_front_T1_01_r, swing1LimitDegrees=45, swing2LimitDegrees=45, twistLimitDegrees=5}
 {bone1Name=dyn_skirt_front_T1_03_r, bone2Name=dyn_skirt_front_T1_02_r, swing1LimitDegrees=45, swing2LimitDegrees=45, twistLimitDegrees=5}
 {bone1Name=dyn_skirt_front_T1_04_r, bone2Name=dyn_skirt_front_T1_03_r, swing1LimitDegrees=45, swing2LimitDegrees=45, twistLimitDegrees=5}
 {bone1Name=dyn_skirt_side_T1_01_r, bone2Name=thigh_r, swing1LimitDegrees=45, swing2LimitDegrees=45, twistLimitDegrees=5}
 {bone1Name=dyn_skirt_side_T1_02_r, bone2Name=dyn_skirt_side_T1_01_r, swing1LimitDegrees=45, swing2LimitDegrees=45, twistLimitDegrees=5}
 {bone1Name=dyn_skirt_side_T1_03_r, bone2Name=dyn_skirt_side_T1_02_r, swing1LimitDegrees=45, swing2LimitDegrees=45, twistLimitDegrees=5}
 {bone1Name=dyn_skirt_side_T1_04_r, bone2Name=dyn_skirt_side_T1_03_r, swing1LimitDegrees=45, swing2LimitDegrees=45, twistLimitDegrees=5}
 {bone1Name=dyn_skirt_back_side_T1_01_r, bone2Name=thigh_r, swing1LimitDegrees=45, swing2LimitDegrees=45, twistLimitDegrees=5}
 {
  bone1Name=dyn_skirt_back_side_T1_02_r
  bone2Name=dyn_skirt_back_side_T1_01_r
  swing1LimitDegrees=45
  swing2LimitDegrees=45
  twistLimitDegrees=5
 }
 {
  bone1Name=dyn_skirt_back_side_T1_03_r
  bone2Name=dyn_skirt_back_side_T1_02_r
  swing1LimitDegrees=45
  swing2LimitDegrees=45
  twistLimitDegrees=5
 }
 {
  bone1Name=dyn_skirt_back_side_T1_04_r
  bone2Name=dyn_skirt_back_side_T1_03_r
  swing1LimitDegrees=45
  swing2LimitDegrees=45
  twistLimitDegrees=5
 }
 {bone1Name=dyn_skirt_back_T1_01_r, bone2Name=thigh_r, swing1LimitDegrees=45, swing2LimitDegrees=45, twistLimitDegrees=5}
 {bone1Name=dyn_skirt_back_T1_02_r, bone2Name=dyn_skirt_back_T1_01_r, swing1LimitDegrees=45, swing2LimitDegrees=45, twistLimitDegrees=5}
 {bone1Name=dyn_skirt_back_T1_03_r, bone2Name=dyn_skirt_back_T1_02_r, swing1LimitDegrees=45, swing2LimitDegrees=45, twistLimitDegrees=5}
 {bone1Name=dyn_skirt_back_T1_04_r, bone2Name=dyn_skirt_back_T1_03_r, swing1LimitDegrees=45, swing2LimitDegrees=45, twistLimitDegrees=5}
 {bone1Name=dyn_skirt_back_T1_04, bone2Name=dyn_skirt_back_T1_03, swing1LimitDegrees=45, swing2LimitDegrees=45, twistLimitDegrees=5}
 {bone1Name=dyn_skirt_back_T1_02, bone2Name=dyn_skirt_back_T1_01, swing1LimitDegrees=45, swing2LimitDegrees=45, twistLimitDegrees=5}
 {bone1Name=dyn_skirt_back_T1_03, bone2Name=dyn_skirt_back_T1_02, swing1LimitDegrees=45, swing2LimitDegrees=45, twistLimitDegrees=5}
 {
  bone1Name=dyn_skirt_back_side_T1_04_l
  bone2Name=dyn_skirt_back_side_T1_03_l
  swing1LimitDegrees=45
  swing2LimitDegrees=45
  twistLimitDegrees=5
 }
 {bone1Name=dyn_skirt_front_mid_T1_01_l, bone2Name=thigh_l, swing1LimitDegrees=45, swing2LimitDegrees=45, twistLimitDegrees=5}
 {
  bone1Name=dyn_skirt_front_mid_T1_02_l
  bone2Name=dyn_skirt_front_mid_T1_01_l
  swing1LimitDegrees=45
  swing2LimitDegrees=45
  twistLimitDegrees=5
 }
 {
  bone1Name=dyn_skirt_front_mid_T1_03_l
  bone2Name=dyn_skirt_front_mid_T1_02_l
  swing1LimitDegrees=45
  swing2LimitDegrees=45
  twistLimitDegrees=5
 }
 {
  bone1Name=dyn_skirt_front_mid_T1_04_l
  bone2Name=dyn_skirt_front_mid_T1_03_l
  swing1LimitDegrees=45
  swing2LimitDegrees=45
  twistLimitDegrees=5
 }
 {bone1Name=dyn_skirt_front_T1_01_l, bone2Name=thigh_l, swing1LimitDegrees=45, swing2LimitDegrees=45, twistLimitDegrees=5}
 {bone1Name=dyn_skirt_front_T1_02_l, bone2Name=dyn_skirt_front_T1_01_l, swing1LimitDegrees=45, swing2LimitDegrees=45, twistLimitDegrees=5}
 {bone1Name=dyn_skirt_front_T1_03_l, bone2Name=dyn_skirt_front_T1_02_l, swing1LimitDegrees=45, swing2LimitDegrees=45, twistLimitDegrees=5}
 {bone1Name=dyn_skirt_front_T1_04_l, bone2Name=dyn_skirt_front_T1_03_l, swing1LimitDegrees=45, swing2LimitDegrees=45, twistLimitDegrees=5}
 {bone1Name=dyn_skirt_back_T1_01_l, bone2Name=thigh_l, swing1LimitDegrees=45, swing2LimitDegrees=45, twistLimitDegrees=5}
 {bone1Name=dyn_skirt_back_T1_02_l, bone2Name=dyn_skirt_back_T1_01_l, swing1LimitDegrees=45, swing2LimitDegrees=45, twistLimitDegrees=5}
 {bone1Name=dyn_skirt_back_T1_03_l, bone2Name=dyn_skirt_back_T1_02_l, swing1LimitDegrees=45, swing2LimitDegrees=45, twistLimitDegrees=5}
 {bone1Name=dyn_skirt_back_T1_04_l, bone2Name=dyn_skirt_back_T1_03_l, swing1LimitDegrees=45, swing2LimitDegrees=45, twistLimitDegrees=5}
 {bone1Name=dyn_skirt_side_T1_01_l, bone2Name=thigh_l, swing1LimitDegrees=45, swing2LimitDegrees=45, twistLimitDegrees=5}
 {bone1Name=dyn_skirt_side_T1_02_l, bone2Name=dyn_skirt_side_T1_01_l, swing1LimitDegrees=45, swing2LimitDegrees=45, twistLimitDegrees=5}
 {bone1Name=dyn_skirt_side_T1_03_l, bone2Name=dyn_skirt_side_T1_02_l, swing1LimitDegrees=45, swing2LimitDegrees=45, twistLimitDegrees=5}
 {bone1Name=dyn_skirt_side_T1_04_l, bone2Name=dyn_skirt_side_T1_03_l, swing1LimitDegrees=45, swing2LimitDegrees=45, twistLimitDegrees=5}
 {bone1Name=dyn_skirt_back_side_T1_01_l, bone2Name=thigh_l, swing1LimitDegrees=45, swing2LimitDegrees=45, twistLimitDegrees=5}
 {
  bone1Name=dyn_skirt_back_side_T1_02_l
  bone2Name=dyn_skirt_back_side_T1_01_l
  swing1LimitDegrees=45
  swing2LimitDegrees=45
  twistLimitDegrees=5
 }
 {
  bone1Name=dyn_skirt_back_side_T1_03_l
  bone2Name=dyn_skirt_back_side_T1_02_l
  swing1LimitDegrees=45
  swing2LimitDegrees=45
  twistLimitDegrees=5
 }
 {bone1Name=thigh_l, bone2Name=pelvis, swing1LimitDegrees=45, swing2LimitDegrees=45, twistLimitDegrees=5}
 {bone1Name=thigh_r, bone2Name=pelvis, swing1LimitDegrees=45, swing2LimitDegrees=45, twistLimitDegrees=5}
 {bone1Name=dyn_skirt_back_T1_01, bone2Name=pelvis, swing1LimitDegrees=45, swing2LimitDegrees=45, twistLimitDegrees=5}
]
