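# Post-process style AnimBlueprint for the T1 Pirate progressive male body, adding secondary motion to its skirt and belt bones.
# It copies the attached parent's pose, then runs per-bone AnimDynamics on the skirt, pushed by skydiving external force.
# Flap nodes drive two right-back skirt bones from calf_r rotation; the belt gets a constraint, LookAt and lower-gravity AnimDynamics.
# The RigidBody node is unconnected, and EventGraph does nothing.
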AnimBlueprint /BRCosmetics/Characters/Player/Male/Medium/Bodies/M_MED_Pirate_Progressive/Meshes/M_MED_Pirate_T1_Body_AnimBP
target_skeleton=&/M_MED_Pirate_T1_Body_Skeleton
graphs [
 AnimGraph {nodes [
   RigidBody() {
    bClampLinearTranslationLimitToRefPose=true
    bOverrideWorldGravity=true
    overridePhysicsAsset=/Game/Characters/Player/Male/Medium/Bodies/M_MED_Pirate_Progressive/Meshes/M_MED_Pirate_T1_Skirt_Physics
    simulationSpace=BaseBoneSpace
    alphaBoolBlend {blendOption=Linear}
    baseBoneRef {boneName=pelvis}
    componentAppliedLinearAccClamp=<100 100 100>
    componentLinearAccScale {z=1}
    componentLinearVelScale {z=1}
    overrideWorldGravity {z=-980}
   }
   n1 = Break Fort Anim Input Skydiving External Force(FortAnimInput_SkydivingExternalForce=Get SkydivingExternalForceInput()) -> (
    bUseSkydivingVectorForce=~
    SkydivingVectorForce=~
    HeadToPelvisDirection=~
    FloatingMultiplier=~
    FloatingAdditive=~
    DivingMultiplier=~
    DivingAdditive=~
    ParachutingMultiplier=~
    ParachutingAdditive=~
   )
   n2 = Anim Dynamics - Bone: dyn_skirt_front_mid_T1_01_l(ComponentPose=Local To Component(LocalPose=Copy Pose From Mesh() {bUseAttachedParent=true}), Alpha=1, ExternalForce=@n1.SkydivingVectorForce) {
    angularDampingOverride=1
    angularSpringConstant=1500
    bGravityOverrideInSimSpace=true
    bOverrideAngularDamping=true
    simulationSpace=RootRelative
    alphaBoolBlend {blendOption=Linear}
    boundBone {boneName=dyn_skirt_front_mid_T1_01_l}
    physicsBodyDefinitions [{
      sphereCollisionRadius=0
      boundBone {boneName=dyn_skirt_front_mid_T1_01_l}
      boxExtents=<5 20 20>
      constraintSetup {angularLimitsMax=<0 5 15>, angularLimitsMin=<0 -5 0>, angularTarget=<0 0 0>}
      localJointOffset=<5 0 0>
     }]
   }
   n3 = Anim Dynamics - Bone: dyn_skirt_front_mid_T1_02_l(ComponentPose=@n2, Alpha=1, ExternalForce=@n1.SkydivingVectorForce) {
    angularDampingOverride=1
    angularSpringConstant=1500
    bGravityOverrideInSimSpace=true
    bOverrideAngularDamping=true
    simulationSpace=RootRelative
    alphaBoolBlend {blendOption=Linear}
    boundBone {boneName=dyn_skirt_front_mid_T1_02_l}
    physicsBodyDefinitions [{
      sphereCollisionRadius=0
      boundBone {boneName=dyn_skirt_front_mid_T1_02_l}
      boxExtents=<5 20 20>
      constraintSetup {angularLimitsMax=<0 5 15>, angularLimitsMin=<0 -5 0>, angularTarget=<0 0 0>}
      localJointOffset=<5 0 0>
     }]
   }
   n4 = Anim Dynamics - Bone: dyn_skirt_front_mid_T1_03_l(ComponentPose=@n3, Alpha=1, ExternalForce=@n1.SkydivingVectorForce) {
    angularDampingOverride=1
    angularSpringConstant=1500
    bGravityOverrideInSimSpace=true
    bOverrideAngularDamping=true
    simulationSpace=RootRelative
    alphaBoolBlend {blendOption=Linear}
    boundBone {boneName=dyn_skirt_front_mid_T1_03_l}
    physicsBodyDefinitions [{
      sphereCollisionRadius=0
      boundBone {boneName=dyn_skirt_front_mid_T1_03_l}
      boxExtents=<5 20 20>
      constraintSetup {angularLimitsMax=<0 5 15>, angularLimitsMin=<0 -5 0>, angularTarget=<0 0 0>}
      localJointOffset=<5 0 0>
     }]
   }
   n5 = Anim Dynamics - Bone: dyn_skirt_front_mid_T1_04_l(ComponentPose=@n4, Alpha=1, ExternalForce=@n1.SkydivingVectorForce) {
    angularDampingOverride=1
    angularSpringConstant=1500
    bGravityOverrideInSimSpace=true
    bOverrideAngularDamping=true
    simulationSpace=RootRelative
    alphaBoolBlend {blendOption=Linear}
    boundBone {boneName=dyn_skirt_front_mid_T1_04_l}
    physicsBodyDefinitions [{
      sphereCollisionRadius=0
      boundBone {boneName=dyn_skirt_front_mid_T1_04_l}
      boxExtents=<5 20 20>
      constraintSetup {angularLimitsMax=<0 5 15>, angularLimitsMin=<0 -5 0>, angularTarget=<0 0 0>}
      localJointOffset=<5 0 0>
     }]
   }
   n6 = Anim Dynamics - Bone: dyn_skirt_front_T1_01_l(ComponentPose=@n5, Alpha=1, ExternalForce=@n1.SkydivingVectorForce) {
    angularDampingOverride=1
    angularSpringConstant=1500
    bGravityOverrideInSimSpace=true
    bOverrideAngularDamping=true
    simulationSpace=RootRelative
    alphaBoolBlend {blendOption=Linear}
    boundBone {boneName=dyn_skirt_front_T1_01_l}
    physicsBodyDefinitions [{
      sphereCollisionRadius=0
      boundBone {boneName=dyn_skirt_front_T1_01_l}
      boxExtents=<5 20 20>
      constraintSetup {angularLimitsMax=<0 5 15>, angularLimitsMin=<0 -5 0>, angularTarget=<0 0 0>}
      localJointOffset=<5 0 0>
     }]
   }
   n7 = Anim Dynamics - Bone: dyn_skirt_front_T1_02_l(ComponentPose=@n6, Alpha=1, ExternalForce=@n1.SkydivingVectorForce) {
    angularDampingOverride=1
    angularSpringConstant=1500
    bGravityOverrideInSimSpace=true
    bOverrideAngularDamping=true
    simulationSpace=RootRelative
    alphaBoolBlend {blendOption=Linear}
    boundBone {boneName=dyn_skirt_front_T1_02_l}
    physicsBodyDefinitions [{
      sphereCollisionRadius=0
      boundBone {boneName=dyn_skirt_front_T1_02_l}
      boxExtents=<5 20 20>
      constraintSetup {angularLimitsMax=<0 5 15>, angularLimitsMin=<0 -5 0>, angularTarget=<0 0 0>}
      localJointOffset=<5 0 0>
     }]
   }
   n8 = Anim Dynamics - Bone: dyn_skirt_front_T1_03_l(ComponentPose=@n7, Alpha=1, ExternalForce=@n1.SkydivingVectorForce) {
    angularDampingOverride=1
    angularSpringConstant=1500
    bGravityOverrideInSimSpace=true
    bOverrideAngularDamping=true
    simulationSpace=RootRelative
    alphaBoolBlend {blendOption=Linear}
    boundBone {boneName=dyn_skirt_front_T1_03_l}
    physicsBodyDefinitions [{
      sphereCollisionRadius=0
      boundBone {boneName=dyn_skirt_front_T1_03_l}
      boxExtents=<5 20 20>
      constraintSetup {angularLimitsMax=<0 5 15>, angularLimitsMin=<0 -5 0>, angularTarget=<0 0 0>}
      localJointOffset=<5 0 0>
     }]
   }
   n9 = Anim Dynamics - Bone: dyn_skirt_front_T1_04_l(ComponentPose=@n8, Alpha=1, ExternalForce=@n1.SkydivingVectorForce) {
    angularDampingOverride=1
    angularSpringConstant=1500
    bGravityOverrideInSimSpace=true
    bOverrideAngularDamping=true
    simulationSpace=RootRelative
    alphaBoolBlend {blendOption=Linear}
    boundBone {boneName=dyn_skirt_front_T1_04_l}
    physicsBodyDefinitions [{
      sphereCollisionRadius=0
      boundBone {boneName=dyn_skirt_front_T1_04_l}
      boxExtents=<5 20 20>
      constraintSetup {angularLimitsMax=<0 5 15>, angularLimitsMin=<0 -5 0>, angularTarget=<0 0 0>}
      localJointOffset=<5 0 0>
     }]
   }
   n10 = Anim Dynamics - Bone: dyn_skirt_back_T1_01_l(ComponentPose=@n9, Alpha=1, ExternalForce=@n1.SkydivingVectorForce) {
    angularDampingOverride=1
    angularSpringConstant=1500
    bGravityOverrideInSimSpace=true
    bOverrideAngularDamping=true
    simulationSpace=RootRelative
    alphaBoolBlend {blendOption=Linear}
    boundBone {boneName=dyn_skirt_back_T1_01_l}
    physicsBodyDefinitions [{
      sphereCollisionRadius=0
      boundBone {boneName=dyn_skirt_back_T1_01_l}
      boxExtents=<5 20 20>
      constraintSetup {angularLimitsMax=<0 5 15>, angularLimitsMin=<0 -5 0>, angularTarget=<0 0 0>}
      localJointOffset=<5 0 0>
     }]
   }
   n11 = Anim Dynamics - Bone: dyn_skirt_back_T1_02_l(ComponentPose=@n10, Alpha=1, ExternalForce=@n1.SkydivingVectorForce) {
    angularDampingOverride=1
    angularSpringConstant=1500
    bGravityOverrideInSimSpace=true
    bOverrideAngularDamping=true
    simulationSpace=RootRelative
    alphaBoolBlend {blendOption=Linear}
    boundBone {boneName=dyn_skirt_back_T1_02_l}
    physicsBodyDefinitions [{
      sphereCollisionRadius=0
      boundBone {boneName=dyn_skirt_back_T1_02_l}
      boxExtents=<5 20 20>
      constraintSetup {angularLimitsMax=<0 5 15>, angularLimitsMin=<0 -5 0>, angularTarget=<0 0 0>}
      localJointOffset=<5 0 0>
     }]
   }
   n12 = Anim Dynamics - Bone: dyn_skirt_back_T1_03_l(ComponentPose=@n11, Alpha=1, ExternalForce=@n1.SkydivingVectorForce) {
    angularDampingOverride=1
    angularSpringConstant=1500
    bGravityOverrideInSimSpace=true
    bOverrideAngularDamping=true
    simulationSpace=RootRelative
    alphaBoolBlend {blendOption=Linear}
    boundBone {boneName=dyn_skirt_back_T1_03_l}
    physicsBodyDefinitions [{
      sphereCollisionRadius=0
      boundBone {boneName=dyn_skirt_back_T1_03_l}
      boxExtents=<5 20 20>
      constraintSetup {angularLimitsMax=<0 5 15>, angularLimitsMin=<0 -5 0>, angularTarget=<0 0 0>}
      localJointOffset=<5 0 0>
     }]
   }
   n13 = Anim Dynamics - Bone: dyn_skirt_back_T1_04_l(ComponentPose=@n12, Alpha=1, ExternalForce=@n1.SkydivingVectorForce) {
    angularDampingOverride=1
    angularSpringConstant=1500
    bGravityOverrideInSimSpace=true
    bOverrideAngularDamping=true
    simulationSpace=RootRelative
    alphaBoolBlend {blendOption=Linear}
    boundBone {boneName=dyn_skirt_back_T1_04_l}
    physicsBodyDefinitions [{
      sphereCollisionRadius=0
      boundBone {boneName=dyn_skirt_back_T1_04_l}
      boxExtents=<5 20 20>
      constraintSetup {angularLimitsMax=<0 5 15>, angularLimitsMin=<0 -5 0>, angularTarget=<0 0 0>}
      localJointOffset=<5 0 0>
     }]
   }
   n14 = Anim Dynamics - Bone: dyn_skirt_side_T1_01_l(ComponentPose=@n13, Alpha=1, ExternalForce=@n1.SkydivingVectorForce) {
    angularDampingOverride=1
    angularSpringConstant=1500
    bGravityOverrideInSimSpace=true
    bOverrideAngularDamping=true
    simulationSpace=RootRelative
    alphaBoolBlend {blendOption=Linear}
    boundBone {boneName=dyn_skirt_side_T1_01_l}
    physicsBodyDefinitions [{
      sphereCollisionRadius=0
      boundBone {boneName=dyn_skirt_side_T1_01_l}
      boxExtents=<5 20 20>
      constraintSetup {angularLimitsMax=<0 5 15>, angularLimitsMin=<0 -5 0>, angularTarget=<0 0 0>}
      localJointOffset=<5 0 0>
     }]
   }
   n15 = Anim Dynamics - Bone: dyn_skirt_side_T1_02_l(ComponentPose=@n14, Alpha=1, ExternalForce=@n1.SkydivingVectorForce) {
    angularDampingOverride=1
    angularSpringConstant=1500
    bGravityOverrideInSimSpace=true
    bOverrideAngularDamping=true
    simulationSpace=RootRelative
    alphaBoolBlend {blendOption=Linear}
    boundBone {boneName=dyn_skirt_side_T1_02_l}
    physicsBodyDefinitions [{
      sphereCollisionRadius=0
      boundBone {boneName=dyn_skirt_side_T1_02_l}
      boxExtents=<5 20 20>
      constraintSetup {angularLimitsMax=<0 5 15>, angularLimitsMin=<0 -5 0>, angularTarget=<0 0 0>}
      localJointOffset=<5 0 0>
     }]
   }
   n16 = Anim Dynamics - Bone: dyn_skirt_side_T1_03_l(ComponentPose=@n15, Alpha=1, ExternalForce=@n1.SkydivingVectorForce) {
    angularDampingOverride=1
    angularSpringConstant=1500
    bGravityOverrideInSimSpace=true
    bOverrideAngularDamping=true
    simulationSpace=RootRelative
    alphaBoolBlend {blendOption=Linear}
    boundBone {boneName=dyn_skirt_side_T1_03_l}
    physicsBodyDefinitions [{
      sphereCollisionRadius=0
      boundBone {boneName=dyn_skirt_side_T1_03_l}
      boxExtents=<5 20 20>
      constraintSetup {angularLimitsMax=<0 5 15>, angularLimitsMin=<0 -5 0>, angularTarget=<0 0 0>}
      localJointOffset=<5 0 0>
     }]
   }
   n17 = Anim Dynamics - Bone: dyn_skirt_side_T1_04_l(ComponentPose=@n16, Alpha=1, ExternalForce=@n1.SkydivingVectorForce) {
    angularDampingOverride=1
    angularSpringConstant=1500
    bGravityOverrideInSimSpace=true
    bOverrideAngularDamping=true
    simulationSpace=RootRelative
    alphaBoolBlend {blendOption=Linear}
    boundBone {boneName=dyn_skirt_side_T1_04_l}
    physicsBodyDefinitions [{
      sphereCollisionRadius=0
      boundBone {boneName=dyn_skirt_side_T1_04_l}
      boxExtents=<5 20 20>
      constraintSetup {angularLimitsMax=<0 5 15>, angularLimitsMin=<0 -5 0>, angularTarget=<0 0 0>}
      localJointOffset=<5 0 0>
     }]
   }
   n18 = Anim Dynamics - Bone: dyn_skirt_back_side_T1_01_l(ComponentPose=@n17, Alpha=1, ExternalForce=@n1.SkydivingVectorForce) {
    angularDampingOverride=1
    angularSpringConstant=1500
    bGravityOverrideInSimSpace=true
    bOverrideAngularDamping=true
    simulationSpace=RootRelative
    alphaBoolBlend {blendOption=Linear}
    boundBone {boneName=dyn_skirt_back_side_T1_01_l}
    physicsBodyDefinitions [{
      sphereCollisionRadius=0
      boundBone {boneName=dyn_skirt_back_side_T1_01_l}
      boxExtents=<5 20 20>
      constraintSetup {angularLimitsMax=<0 5 15>, angularLimitsMin=<0 -5 0>, angularTarget=<0 0 0>}
      localJointOffset=<5 0 0>
     }]
   }
   n19 = Anim Dynamics - Bone: dyn_skirt_back_side_T1_02_l(ComponentPose=@n18, Alpha=1, ExternalForce=@n1.SkydivingVectorForce) {
    angularDampingOverride=1
    angularSpringConstant=1500
    bGravityOverrideInSimSpace=true
    bOverrideAngularDamping=true
    simulationSpace=RootRelative
    alphaBoolBlend {blendOption=Linear}
    boundBone {boneName=dyn_skirt_back_side_T1_02_l}
    physicsBodyDefinitions [{
      sphereCollisionRadius=0
      boundBone {boneName=dyn_skirt_back_side_T1_02_l}
      boxExtents=<5 20 20>
      constraintSetup {angularLimitsMax=<0 5 15>, angularLimitsMin=<0 -5 0>, angularTarget=<0 0 0>}
      localJointOffset=<5 0 0>
     }]
   }
   n20 = Anim Dynamics - Bone: dyn_skirt_back_side_T1_03_l(ComponentPose=@n19, Alpha=1, ExternalForce=@n1.SkydivingVectorForce) {
    angularDampingOverride=1
    angularSpringConstant=1500
    bGravityOverrideInSimSpace=true
    bOverrideAngularDamping=true
    simulationSpace=RootRelative
    alphaBoolBlend {blendOption=Linear}
    boundBone {boneName=dyn_skirt_back_side_T1_03_l}
    physicsBodyDefinitions [{
      sphereCollisionRadius=0
      boundBone {boneName=dyn_skirt_back_side_T1_03_l}
      boxExtents=<5 20 20>
      constraintSetup {angularLimitsMax=<0 5 15>, angularLimitsMin=<0 -5 0>, angularTarget=<0 0 0>}
      localJointOffset=<5 0 0>
     }]
   }
   n21 = Anim Dynamics - Bone: dyn_skirt_back_side_T1_04_l(ComponentPose=@n20, Alpha=1, ExternalForce=@n1.SkydivingVectorForce) {
    angularDampingOverride=1
    angularSpringConstant=1500
    bGravityOverrideInSimSpace=true
    bOverrideAngularDamping=true
    simulationSpace=RootRelative
    alphaBoolBlend {blendOption=Linear}
    boundBone {boneName=dyn_skirt_back_side_T1_04_l}
    physicsBodyDefinitions [{
      sphereCollisionRadius=0
      boundBone {boneName=dyn_skirt_back_side_T1_04_l}
      boxExtents=<5 20 20>
      constraintSetup {angularLimitsMax=<0 5 15>, angularLimitsMin=<0 -5 0>, angularTarget=<0 0 0>}
      localJointOffset=<5 0 0>
     }]
   }
   n22 = Anim Dynamics - Bone: dyn_skirt_front_mid_T1_01_r(ComponentPose=@n21, Alpha=1, ExternalForce=@n1.SkydivingVectorForce) {
    angularDampingOverride=1
    angularSpringConstant=1500
    bGravityOverrideInSimSpace=true
    bOverrideAngularDamping=true
    simulationSpace=RootRelative
    alphaBoolBlend {blendOption=Linear}
    boundBone {boneName=dyn_skirt_front_mid_T1_01_r}
    physicsBodyDefinitions [{
      sphereCollisionRadius=0
      boundBone {boneName=dyn_skirt_front_mid_T1_01_r}
      boxExtents=<5 20 20>
      constraintSetup {angularLimitsMax=<0 5 15>, angularLimitsMin=<0 -5 0>, angularTarget=<0 0 0>}
      localJointOffset=<5 0 0>
     }]
   }
   n23 = Anim Dynamics - Bone: dyn_skirt_front_mid_T1_02_r(ComponentPose=@n22, Alpha=1, ExternalForce=@n1.SkydivingVectorForce) {
    angularDampingOverride=1
    angularSpringConstant=1500
    bGravityOverrideInSimSpace=true
    bOverrideAngularDamping=true
    simulationSpace=RootRelative
    alphaBoolBlend {blendOption=Linear}
    boundBone {boneName=dyn_skirt_front_mid_T1_02_r}
    physicsBodyDefinitions [{
      sphereCollisionRadius=0
      boundBone {boneName=dyn_skirt_front_mid_T1_02_r}
      boxExtents=<5 20 20>
      constraintSetup {angularLimitsMax=<0 5 15>, angularLimitsMin=<0 -5 0>, angularTarget=<0 0 0>}
      localJointOffset=<5 0 0>
     }]
   }
   n24 = Anim Dynamics - Bone: dyn_skirt_front_mid_T1_03_r(ComponentPose=@n23, Alpha=1, ExternalForce=@n1.SkydivingVectorForce) {
    angularDampingOverride=1
    angularSpringConstant=1500
    bGravityOverrideInSimSpace=true
    bOverrideAngularDamping=true
    simulationSpace=RootRelative
    alphaBoolBlend {blendOption=Linear}
    boundBone {boneName=dyn_skirt_front_mid_T1_03_r}
    physicsBodyDefinitions [{
      sphereCollisionRadius=0
      boundBone {boneName=dyn_skirt_front_mid_T1_03_r}
      boxExtents=<5 20 20>
      constraintSetup {angularLimitsMax=<0 5 15>, angularLimitsMin=<0 -5 0>, angularTarget=<0 0 0>}
      localJointOffset=<5 0 0>
     }]
   }
   n25 = Anim Dynamics - Bone: dyn_skirt_front_mid_T1_04_r(ComponentPose=@n24, Alpha=1, ExternalForce=@n1.SkydivingVectorForce) {
    angularDampingOverride=1
    angularSpringConstant=1500
    bGravityOverrideInSimSpace=true
    bOverrideAngularDamping=true
    simulationSpace=RootRelative
    alphaBoolBlend {blendOption=Linear}
    boundBone {boneName=dyn_skirt_front_mid_T1_04_r}
    physicsBodyDefinitions [{
      sphereCollisionRadius=0
      boundBone {boneName=dyn_skirt_front_mid_T1_04_r}
      boxExtents=<5 20 20>
      constraintSetup {angularLimitsMax=<0 5 15>, angularLimitsMin=<0 -5 0>, angularTarget=<0 0 0>}
      localJointOffset=<5 0 0>
     }]
   }
   n26 = Anim Dynamics - Bone: dyn_skirt_front_T1_01_r(ComponentPose=@n25, Alpha=1, ExternalForce=@n1.SkydivingVectorForce) {
    angularDampingOverride=1
    angularSpringConstant=1500
    bGravityOverrideInSimSpace=true
    bOverrideAngularDamping=true
    simulationSpace=RootRelative
    alphaBoolBlend {blendOption=Linear}
    boundBone {boneName=dyn_skirt_front_T1_01_r}
    physicsBodyDefinitions [{
      sphereCollisionRadius=0
      boundBone {boneName=dyn_skirt_front_T1_01_r}
      boxExtents=<5 20 20>
      constraintSetup {angularLimitsMax=<0 5 15>, angularLimitsMin=<0 -5 0>, angularTarget=<0 0 0>}
      localJointOffset=<5 0 0>
     }]
   }
   n27 = Anim Dynamics - Bone: dyn_skirt_front_T1_02_r(ComponentPose=@n26, Alpha=1, ExternalForce=@n1.SkydivingVectorForce) {
    angularDampingOverride=1
    angularSpringConstant=1500
    bGravityOverrideInSimSpace=true
    bOverrideAngularDamping=true
    simulationSpace=RootRelative
    alphaBoolBlend {blendOption=Linear}
    boundBone {boneName=dyn_skirt_front_T1_02_r}
    physicsBodyDefinitions [{
      sphereCollisionRadius=0
      boundBone {boneName=dyn_skirt_front_T1_02_r}
      boxExtents=<5 20 20>
      constraintSetup {angularLimitsMax=<0 5 15>, angularLimitsMin=<0 -5 0>, angularTarget=<0 0 0>}
      localJointOffset=<5 0 0>
     }]
   }
   n28 = Anim Dynamics - Bone: dyn_skirt_front_T1_03_r(ComponentPose=@n27, Alpha=1, ExternalForce=@n1.SkydivingVectorForce) {
    angularDampingOverride=1
    angularSpringConstant=1500
    bGravityOverrideInSimSpace=true
    bOverrideAngularDamping=true
    simulationSpace=RootRelative
    alphaBoolBlend {blendOption=Linear}
    boundBone {boneName=dyn_skirt_front_T1_03_r}
    physicsBodyDefinitions [{
      sphereCollisionRadius=0
      boundBone {boneName=dyn_skirt_front_T1_03_r}
      boxExtents=<5 20 20>
      constraintSetup {angularLimitsMax=<0 5 15>, angularLimitsMin=<0 -5 0>, angularTarget=<0 0 0>}
      localJointOffset=<5 0 0>
     }]
   }
   n29 = Anim Dynamics - Bone: dyn_skirt_front_T1_04_r(ComponentPose=@n28, Alpha=1, ExternalForce=@n1.SkydivingVectorForce) {
    angularDampingOverride=1
    angularSpringConstant=1500
    bGravityOverrideInSimSpace=true
    bOverrideAngularDamping=true
    simulationSpace=RootRelative
    alphaBoolBlend {blendOption=Linear}
    boundBone {boneName=dyn_skirt_front_T1_04_r}
    physicsBodyDefinitions [{
      sphereCollisionRadius=0
      boundBone {boneName=dyn_skirt_front_T1_04_r}
      boxExtents=<5 20 20>
      constraintSetup {angularLimitsMax=<0 5 15>, angularLimitsMin=<0 -5 0>, angularTarget=<0 0 0>}
      localJointOffset=<5 0 0>
     }]
   }
   n30 = Anim Dynamics - Bone: dyn_skirt_side_T1_01_r(ComponentPose=@n29, Alpha=1, ExternalForce=@n1.SkydivingVectorForce) {
    angularDampingOverride=1
    angularSpringConstant=1500
    bGravityOverrideInSimSpace=true
    bOverrideAngularDamping=true
    simulationSpace=RootRelative
    alphaBoolBlend {blendOption=Linear}
    boundBone {boneName=dyn_skirt_side_T1_01_r}
    physicsBodyDefinitions [{
      sphereCollisionRadius=0
      boundBone {boneName=dyn_skirt_side_T1_01_r}
      boxExtents=<5 20 20>
      constraintSetup {angularLimitsMax=<0 5 15>, angularLimitsMin=<0 -5 0>, angularTarget=<0 0 0>}
      localJointOffset=<5 0 0>
     }]
   }
   n31 = Anim Dynamics - Bone: dyn_skirt_side_T1_02_r(ComponentPose=@n30, Alpha=1, ExternalForce=@n1.SkydivingVectorForce) {
    angularDampingOverride=1
    angularSpringConstant=1500
    bGravityOverrideInSimSpace=true
    bOverrideAngularDamping=true
    simulationSpace=RootRelative
    alphaBoolBlend {blendOption=Linear}
    boundBone {boneName=dyn_skirt_side_T1_02_r}
    physicsBodyDefinitions [{
      sphereCollisionRadius=0
      boundBone {boneName=dyn_skirt_side_T1_02_r}
      boxExtents=<5 20 20>
      constraintSetup {angularLimitsMax=<0 5 15>, angularLimitsMin=<0 -5 0>, angularTarget=<0 0 0>}
      localJointOffset=<5 0 0>
     }]
   }
   n32 = Anim Dynamics - Bone: dyn_skirt_side_T1_03_r(ComponentPose=@n31, Alpha=1, ExternalForce=@n1.SkydivingVectorForce) {
    angularDampingOverride=1
    angularSpringConstant=1500
    bGravityOverrideInSimSpace=true
    bOverrideAngularDamping=true
    simulationSpace=RootRelative
    alphaBoolBlend {blendOption=Linear}
    boundBone {boneName=dyn_skirt_side_T1_03_r}
    physicsBodyDefinitions [{
      sphereCollisionRadius=0
      boundBone {boneName=dyn_skirt_side_T1_03_r}
      boxExtents=<5 20 20>
      constraintSetup {angularLimitsMax=<0 5 15>, angularLimitsMin=<0 -5 0>, angularTarget=<0 0 0>}
      localJointOffset=<5 0 0>
     }]
   }
   n33 = Anim Dynamics - Bone: dyn_skirt_side_T1_04_r(ComponentPose=@n32, Alpha=1, ExternalForce=@n1.SkydivingVectorForce) {
    angularDampingOverride=1
    angularSpringConstant=1500
    bGravityOverrideInSimSpace=true
    bOverrideAngularDamping=true
    simulationSpace=RootRelative
    alphaBoolBlend {blendOption=Linear}
    boundBone {boneName=dyn_skirt_side_T1_04_r}
    physicsBodyDefinitions [{
      sphereCollisionRadius=0
      boundBone {boneName=dyn_skirt_side_T1_04_r}
      boxExtents=<5 20 20>
      constraintSetup {angularLimitsMax=<0 5 15>, angularLimitsMin=<0 -5 0>, angularTarget=<0 0 0>}
      localJointOffset=<5 0 0>
     }]
   }
   n34 = Anim Dynamics - Bone: dyn_skirt_back_side_T1_01_r(ComponentPose=@n33, Alpha=1, ExternalForce=@n1.SkydivingVectorForce) {
    angularDampingOverride=1
    angularSpringConstant=1500
    bGravityOverrideInSimSpace=true
    bOverrideAngularDamping=true
    simulationSpace=RootRelative
    alphaBoolBlend {blendOption=Linear}
    boundBone {boneName=dyn_skirt_back_side_T1_01_r}
    physicsBodyDefinitions [{
      sphereCollisionRadius=0
      boundBone {boneName=dyn_skirt_back_side_T1_01_r}
      boxExtents=<5 20 20>
      constraintSetup {angularLimitsMax=<0 5 15>, angularLimitsMin=<0 -5 0>, angularTarget=<0 0 0>}
      localJointOffset=<5 0 0>
     }]
   }
   n35 = Anim Dynamics - Bone: dyn_skirt_back_side_T1_02_r(ComponentPose=@n34, Alpha=1, ExternalForce=@n1.SkydivingVectorForce) {
    angularDampingOverride=1
    angularSpringConstant=1500
    bGravityOverrideInSimSpace=true
    bOverrideAngularDamping=true
    simulationSpace=RootRelative
    alphaBoolBlend {blendOption=Linear}
    boundBone {boneName=dyn_skirt_back_side_T1_02_r}
    physicsBodyDefinitions [{
      sphereCollisionRadius=0
      boundBone {boneName=dyn_skirt_back_side_T1_02_r}
      boxExtents=<5 20 20>
      constraintSetup {angularLimitsMax=<0 5 15>, angularLimitsMin=<0 -5 0>, angularTarget=<0 0 0>}
      localJointOffset=<5 0 0>
     }]
   }
   n36 = Anim Dynamics - Bone: dyn_skirt_back_side_T1_03_r(ComponentPose=@n35, Alpha=1, ExternalForce=@n1.SkydivingVectorForce) {
    angularDampingOverride=1
    angularSpringConstant=1500
    bGravityOverrideInSimSpace=true
    bOverrideAngularDamping=true
    simulationSpace=RootRelative
    alphaBoolBlend {blendOption=Linear}
    boundBone {boneName=dyn_skirt_back_side_T1_03_r}
    physicsBodyDefinitions [{
      sphereCollisionRadius=0
      boundBone {boneName=dyn_skirt_back_side_T1_03_r}
      boxExtents=<5 20 20>
      constraintSetup {angularLimitsMax=<0 5 15>, angularLimitsMin=<0 -5 0>, angularTarget=<0 0 0>}
      localJointOffset=<5 0 0>
     }]
   }
   n37 = Anim Dynamics - Bone: dyn_skirt_back_side_T1_04_r(ComponentPose=@n36, Alpha=1, ExternalForce=@n1.SkydivingVectorForce) {
    angularDampingOverride=1
    angularSpringConstant=1500
    bGravityOverrideInSimSpace=true
    bOverrideAngularDamping=true
    simulationSpace=RootRelative
    alphaBoolBlend {blendOption=Linear}
    boundBone {boneName=dyn_skirt_back_side_T1_04_r}
    physicsBodyDefinitions [{
      sphereCollisionRadius=0
      boundBone {boneName=dyn_skirt_back_side_T1_04_r}
      boxExtents=<5 20 20>
      constraintSetup {angularLimitsMax=<0 5 15>, angularLimitsMin=<0 -5 0>, angularTarget=<0 0 0>}
      localJointOffset=<5 0 0>
     }]
   }
   n38 = Anim Dynamics - Bone: dyn_skirt_back_T1_01_r(ComponentPose=@n37, Alpha=1, ExternalForce=@n1.SkydivingVectorForce) {
    angularDampingOverride=1
    angularSpringConstant=1500
    bGravityOverrideInSimSpace=true
    bOverrideAngularDamping=true
    simulationSpace=RootRelative
    alphaBoolBlend {blendOption=Linear}
    boundBone {boneName=dyn_skirt_back_T1_01_r}
    physicsBodyDefinitions [{
      sphereCollisionRadius=0
      boundBone {boneName=dyn_skirt_back_T1_01_r}
      boxExtents=<5 20 20>
      constraintSetup {angularLimitsMax=<0 5 15>, angularLimitsMin=<0 -5 0>, angularTarget=<0 0 0>}
      localJointOffset=<5 0 0>
     }]
   }
   n39 = Anim Dynamics - Bone: dyn_skirt_back_T1_02_r(ComponentPose=@n38, Alpha=1, ExternalForce=@n1.SkydivingVectorForce) {
    angularDampingOverride=1
    angularSpringConstant=1500
    bGravityOverrideInSimSpace=true
    bOverrideAngularDamping=true
    simulationSpace=RootRelative
    alphaBoolBlend {blendOption=Linear}
    boundBone {boneName=dyn_skirt_back_T1_02_r}
    physicsBodyDefinitions [{
      sphereCollisionRadius=0
      boundBone {boneName=dyn_skirt_back_T1_02_r}
      boxExtents=<5 20 20>
      constraintSetup {angularLimitsMax=<0 5 15>, angularLimitsMin=<0 -5 0>, angularTarget=<0 0 0>}
      localJointOffset=<5 0 0>
     }]
   }
   n40 = Anim Dynamics - Bone: dyn_skirt_back_T1_03_r(ComponentPose=@n39, Alpha=1, ExternalForce=@n1.SkydivingVectorForce) {
    angularDampingOverride=1
    angularSpringConstant=1500
    bGravityOverrideInSimSpace=true
    bOverrideAngularDamping=true
    simulationSpace=RootRelative
    alphaBoolBlend {blendOption=Linear}
    boundBone {boneName=dyn_skirt_back_T1_03_r}
    physicsBodyDefinitions [{
      sphereCollisionRadius=0
      boundBone {boneName=dyn_skirt_back_T1_03_r}
      boxExtents=<5 20 20>
      constraintSetup {angularLimitsMax=<0 5 15>, angularLimitsMin=<0 -5 0>, angularTarget=<0 0 0>}
      localJointOffset=<5 0 0>
     }]
   }
   n41 = Anim Dynamics - Bone: dyn_skirt_back_T1_04_r(ComponentPose=@n40, Alpha=1, ExternalForce=@n1.SkydivingVectorForce) {
    angularDampingOverride=1
    angularSpringConstant=1500
    bGravityOverrideInSimSpace=true
    bOverrideAngularDamping=true
    simulationSpace=RootRelative
    alphaBoolBlend {blendOption=Linear}
    boundBone {boneName=dyn_skirt_back_T1_04_r}
    physicsBodyDefinitions [{
      sphereCollisionRadius=0
      boundBone {boneName=dyn_skirt_back_T1_04_r}
      boxExtents=<5 20 20>
      constraintSetup {angularLimitsMax=<0 5 15>, angularLimitsMin=<0 -5 0>, angularTarget=<0 0 0>}
      localJointOffset=<5 0 0>
     }]
   }
   n42 = Anim Dynamics - Bone: dyn_skirt_back_T1_01(ComponentPose=@n41, Alpha=1, ExternalForce=@n1.SkydivingVectorForce) {
    angularDampingOverride=1
    angularSpringConstant=1500
    bGravityOverrideInSimSpace=true
    bOverrideAngularDamping=true
    simulationSpace=RootRelative
    alphaBoolBlend {blendOption=Linear}
    boundBone {boneName=dyn_skirt_back_T1_01}
    physicsBodyDefinitions [{
      sphereCollisionRadius=0
      boundBone {boneName=dyn_skirt_back_T1_01}
      boxExtents=<5 20 20>
      constraintSetup {angularLimitsMax=<0 5 15>, angularLimitsMin=<0 -5 0>, angularTarget=<0 0 0>}
      localJointOffset=<5 0 0>
     }]
   }
   n43 = Anim Dynamics - Bone: dyn_skirt_back_T1_02(ComponentPose=@n42, Alpha=1, ExternalForce=@n1.SkydivingVectorForce) {
    angularDampingOverride=1
    angularSpringConstant=1500
    bGravityOverrideInSimSpace=true
    bOverrideAngularDamping=true
    simulationSpace=RootRelative
    alphaBoolBlend {blendOption=Linear}
    boundBone {boneName=dyn_skirt_back_T1_02}
    physicsBodyDefinitions [{
      sphereCollisionRadius=0
      boundBone {boneName=dyn_skirt_back_T1_02}
      boxExtents=<5 20 20>
      constraintSetup {angularLimitsMax=<0 5 15>, angularLimitsMin=<0 -5 0>, angularTarget=<0 0 0>}
      localJointOffset=<5 0 0>
     }]
   }
   n44 = Anim Dynamics - Bone: dyn_skirt_back_T1_03(ComponentPose=@n43, Alpha=1, ExternalForce=@n1.SkydivingVectorForce) {
    angularDampingOverride=1
    angularSpringConstant=1500
    bGravityOverrideInSimSpace=true
    bOverrideAngularDamping=true
    simulationSpace=RootRelative
    alphaBoolBlend {blendOption=Linear}
    boundBone {boneName=dyn_skirt_back_T1_03}
    physicsBodyDefinitions [{
      sphereCollisionRadius=0
      boundBone {boneName=dyn_skirt_back_T1_03}
      boxExtents=<5 20 20>
      constraintSetup {angularLimitsMax=<0 5 15>, angularLimitsMin=<0 -5 0>, angularTarget=<0 0 0>}
      localJointOffset=<5 0 0>
     }]
   }
   n45 = Anim Dynamics - Bone: dyn_skirt_back_T1_04(ComponentPose=@n44, Alpha=1, ExternalForce=@n1.SkydivingVectorForce) {
    angularDampingOverride=1
    angularSpringConstant=1500
    bGravityOverrideInSimSpace=true
    bOverrideAngularDamping=true
    simulationSpace=RootRelative
    alphaBoolBlend {blendOption=Linear}
    boundBone {boneName=dyn_skirt_back_T1_04}
    physicsBodyDefinitions [{
      sphereCollisionRadius=0
      boundBone {boneName=dyn_skirt_back_T1_04}
      boxExtents=<5 20 20>
      constraintSetup {angularLimitsMax=<0 5 15>, angularLimitsMin=<0 -5 0>, angularTarget=<0 0 0>}
      localJointOffset=<5 0 0>
     }]
   }
   n46 = Flap Node - Bone: dyn_skirt_back_T1_02_r!(ComponentPose:Component Space Pose Link Structure=@n45, Alpha:float=1) -> (Pose:Component Space Pose Link Structure) {
    is=title FortAnimGraphNode_Flap
    bUseClamp=true
    targetClampMax=3.14
    targetClampMin=-3.14
    targetComponent=RotationZ
    alphaBoolBlend {blendOption=Linear}
    sourceDrives [{
      bUseRange=false
      drivingCurve=null
      multiplier=-0.5
      rangeMax=1
      rangeMin=-1
      remappedMax=1
      remappedMin=0
      sourceComponent=RotationZ
      useQuaternion=true
      sourceBone {boneName=calf_r}
     }]
    targetBone {boneName=dyn_skirt_back_T1_02_r}
   }
   n47 = Flap Node - Bone: dyn_skirt_back_side_T1_02_r!(ComponentPose:Component Space Pose Link Structure=@n46, Alpha:float=1) -> (Pose:Component Space Pose Link Structure) {
    is=title FortAnimGraphNode_Flap
    bUseClamp=true
    targetClampMax=3.14
    targetClampMin=-3.14
    targetComponent=RotationZ
    alphaBoolBlend {blendOption=Linear}
    sourceDrives [{
      bUseRange=false
      drivingCurve=null
      multiplier=-0.2
      rangeMax=1
      rangeMin=-1
      remappedMax=1
      remappedMin=0
      sourceComponent=RotationZ
      useQuaternion=true
      sourceBone {boneName=calf_r}
     }]
    targetBone {boneName=dyn_skirt_back_side_T1_02_r}
   }
   n48 = "Constraint - dyn_belt_T1_01 (PPP)"!(ComponentPose=@n47, Alpha=1, ConstraintWeights_0=0.9325, ConstraintWeights_1:float=0, ConstraintWeights_2:float=0.0675) -> (Pose) {
    alphaBoolBlend {blendOption=Linear}
    boneToModify {boneName=dyn_belt_T1_01}
    constraintSetup [{targetBone {boneName=pelvis}}, {targetBone {boneName=dyn_belt_T1_03}}, {targetBone {boneName=spine_01}}]
    constraintWeights [1,1,1]
   }
   n49 = Look At - Bone: dyn_belt_T1_01(ComponentPose=@n48, Alpha=1) {
    alphaBoolBlend {blendOption=Linear}
    boneToModify {boneName=dyn_belt_T1_01}
    lookAtLocation=<0 2>
    lookAtTarget {boneReference {boneName=dyn_skirt_front_mid_T1_03_l}}
    lookAt_Axis {axis=<1 0>}
    lookUp_Axis {axis=<0 1>}
   }
   n50 = Anim Dynamics - Bone: dyn_belt_T1_01(ComponentPose=@n49, Alpha=1, ExternalForce="0,0,0") {
    angularDampingOverride=1
    angularSpringConstant=1500
    bGravityOverrideInSimSpace=true
    bOverrideAngularDamping=true
    gravityScale=0.35
    simulationSpace=RootRelative
    alphaBoolBlend {blendOption=Linear}
    boundBone {boneName=dyn_belt_T1_01}
    componentAppliedLinearAccClamp=<3000 3000 3000>
    componentLinearAccScale=<0.2 0.2 1>
    componentLinearVelScale=<0.2 0.2 1>
    physicsBodyDefinitions [{
      sphereCollisionRadius=0
      boundBone {boneName=dyn_belt_T1_01}
      boxExtents=<5 20 20>
      constraintSetup {angularLimitsMax=<0 20 20>, angularLimitsMin=<0 -20 0>, angularTarget=<0 0 0>}
      localJointOffset=<5 0 0>
     }]
   }
   n51 = Anim Dynamics - Bone: dyn_belt_T1_02(ComponentPose=@n50, Alpha=1, ExternalForce="0,0,0") {
    angularDampingOverride=1
    angularSpringConstant=1500
    bGravityOverrideInSimSpace=true
    bOverrideAngularDamping=true
    gravityScale=0.35
    simulationSpace=RootRelative
    alphaBoolBlend {blendOption=Linear}
    boundBone {boneName=dyn_belt_T1_02}
    componentAppliedLinearAccClamp=<3000 3000 3000>
    componentLinearAccScale=<0.2 0.2 1>
    componentLinearVelScale=<0.2 0.2 1>
    physicsBodyDefinitions [{
      sphereCollisionRadius=0
      boundBone {boneName=dyn_belt_T1_02}
      boxExtents=<5 20 20>
      constraintSetup {angularLimitsMax=<0 20 20>, angularLimitsMin=<0 -20 0>, angularTarget=<0 0 0>}
      localJointOffset=<5 0 0>
     }]
   }
   n52 = Anim Dynamics - Bone: dyn_belt_T1_03(ComponentPose=@n51, Alpha=1, ExternalForce="0,0,0") {
    angularDampingOverride=1
    angularSpringConstant=1500
    bGravityOverrideInSimSpace=true
    bOverrideAngularDamping=true
    gravityScale=0.35
    simulationSpace=RootRelative
    alphaBoolBlend {blendOption=Linear}
    boundBone {boneName=dyn_belt_T1_03}
    componentAppliedLinearAccClamp=<3000 3000 3000>
    componentLinearAccScale=<0.2 0.2 1>
    componentLinearVelScale=<0.2 0.2 1>
    physicsBodyDefinitions [{
      sphereCollisionRadius=0
      boundBone {boneName=dyn_belt_T1_03}
      boxExtents=<5 20 20>
      constraintSetup {angularLimitsMax=<0 20 20>, angularLimitsMin=<0 -20 0>, angularTarget=<0 0 0>}
      localJointOffset=<5 0 0>
     }]
   }
   Output Pose(Result=Component To Local(ComponentPose=@n52)) {layerGroup=None}
  ]}
 EventGraph {nodes [
   TryGetPawnOwner()
   Event BlueprintUpdateAnimation()
  ]}
]
variables []
