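# Physics asset for the tier 3 face accessory of the medium male progressive Pirate body, giving capsule physics to the beard, braids and mustache.
# A two-bone beard chain hangs from the head, with mirrored A braids branching off the first bone and shorter B braids off the second.
# Beard joints and braid roots swing and twist freely.
# Braid segments and the two three-bone mustache chains have tight angular limits.
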PhysicsAsset /BRCosmetics/Characters/Player/Male/Medium/Bodies/M_MED_Pirate_Progressive/Meshes/Parts/M_MED_Pirate_T3_FaceAcc_Physics
bodies {
 dyn_beard_01 [{shapeType=Capsule, length=5, radius=2, center=<2.5 0 0>, rotation=<rot 90 0 0>}]
 dyn_beard_02 [{shapeType=Capsule, length=5, radius=2, center=<2.5 0 0>, rotation=<rot 90 0 0>}]
 dyn_beard_braid_A_T3_01_l [{shapeType=Capsule, length=5, radius=3, center=<2.5 0 0>, rotation=<rot 90 0 0>}]
 dyn_beard_braid_A_T3_01_r [{shapeType=Capsule, length=5, radius=3, center=<-2.5 0 0>, rotation=<rot 90 0 0>}]
 dyn_beard_braid_A_T3_02_l [{shapeType=Capsule, length=5, radius=3, rotation=<rot 90 0 0>}]
 dyn_beard_braid_A_T3_02_r [{shapeType=Capsule, length=5, radius=3, rotation=<rot 90 0 0>}]
 dyn_beard_braid_A_T3_master_l [{shapeType=Capsule, length=5, radius=2, center=<2.5 0 0>, rotation=<rot 90 0 0>}]
 dyn_beard_braid_A_T3_master_r [{shapeType=Capsule, length=5, radius=2, center=<-2.5 0 0>, rotation=<rot 90 0 0>}]
 dyn_beard_braid_B_T3_01_l [{shapeType=Capsule, length=5, radius=3, center=<2.5 0 0>, rotation=<rot 90 0 0>}]
 dyn_beard_braid_B_T3_01_r [{shapeType=Capsule, length=5, radius=3, center=<-2.5 0 0>, rotation=<rot 90 0 0>}]
 dyn_beard_braid_B_T3_master_l [{shapeType=Capsule, length=5, radius=2, center=<2.5 0 0>, rotation=<rot 90 0 0>}]
 dyn_beard_braid_B_T3_master_r [{shapeType=Capsule, length=5, radius=2, center=<-2.5 0 0>, rotation=<rot 90 0 0>}]
 dyn_mustache_T3_01_l [{shapeType=Capsule, length=4.6794, radius=2.915, rotation=<rot 90 0 0>}]
 dyn_mustache_T3_01_r [{shapeType=Capsule, length=4.6798, radius=2.9152, rotation=<rot 90 0 0>}]
 dyn_mustache_T3_02_l [{shapeType=Capsule, length=4.3622, radius=2.0469, rotation=<rot 90 0 0>}]
 dyn_mustache_T3_02_r [{shapeType=Capsule, length=4.3622, radius=2.0469, rotation=<rot 90 0 0>}]
 dyn_mustache_T3_03_l [{shapeType=Capsule, length=2.4982, radius=1.4385, rotation=<rot 90 0 0>}]
 dyn_mustache_T3_03_r [{shapeType=Capsule, length=2.4982, radius=1.4385, rotation=<rot 90 0 0>}]
 head [{shapeType=Capsule, length=1, radius=1}]
}
constraints [
 {
  bone1Name=dyn_beard_02
  bone2Name=dyn_beard_01
  swing1LimitDegrees=45
  swing1Motion=Free
  swing2LimitDegrees=45
  swing2Motion=Free
  twistLimitDegrees=45
  twistMotion=Free
 }
 {
  bone1Name=dyn_beard_braid_B_T3_master_l
  bone2Name=dyn_beard_02
  swing1LimitDegrees=45
  swing1Motion=Free
  swing2LimitDegrees=45
  swing2Motion=Free
  twistLimitDegrees=45
  twistMotion=Free
 }
 {
  bone1Name=dyn_beard_braid_B_T3_master_r
  bone2Name=dyn_beard_02
  swing1LimitDegrees=45
  swing1Motion=Free
  swing2LimitDegrees=45
  swing2Motion=Free
  twistLimitDegrees=45
  twistMotion=Free
 }
 {
  bone1Name=dyn_beard_braid_A_T3_master_l
  bone2Name=dyn_beard_01
  swing1LimitDegrees=45
  swing1Motion=Free
  swing2LimitDegrees=45
  swing2Motion=Free
  twistLimitDegrees=45
  twistMotion=Free
 }
 {
  bone1Name=dyn_beard_braid_A_T3_master_r
  bone2Name=dyn_beard_01
  swing1LimitDegrees=45
  swing1Motion=Free
  swing2LimitDegrees=45
  swing2Motion=Free
  twistLimitDegrees=45
  twistMotion=Free
 }
 {
  bone1Name=dyn_beard_braid_A_T3_01_r
  bone2Name=dyn_beard_braid_A_T3_master_r
  swing1LimitDegrees=10
  swing2LimitDegrees=10
  twistLimitDegrees=1
 }
 {bone1Name=dyn_beard_braid_A_T3_02_r, bone2Name=dyn_beard_braid_A_T3_01_r, swing1LimitDegrees=10, swing2LimitDegrees=10, twistLimitDegrees=1}
 {
  bone1Name=dyn_beard_braid_A_T3_01_l
  bone2Name=dyn_beard_braid_A_T3_master_l
  swing1LimitDegrees=10
  swing2LimitDegrees=10
  twistLimitDegrees=1
 }
 {bone1Name=dyn_beard_braid_A_T3_02_l, bone2Name=dyn_beard_braid_A_T3_01_l, swing1LimitDegrees=10, swing2LimitDegrees=10, twistLimitDegrees=1}
 {
  bone1Name=dyn_beard_braid_B_T3_01_r
  bone2Name=dyn_beard_braid_B_T3_master_r
  swing1LimitDegrees=10
  swing2LimitDegrees=10
  twistLimitDegrees=1
 }
 {
  bone1Name=dyn_beard_braid_B_T3_01_l
  bone2Name=dyn_beard_braid_B_T3_master_l
  swing1LimitDegrees=10
  swing2LimitDegrees=10
  twistLimitDegrees=1
 }
 {
  bone1Name=dyn_beard_01
  bone2Name=head
  swing1LimitDegrees=45
  swing1Motion=Free
  swing2LimitDegrees=45
  swing2Motion=Free
  twistLimitDegrees=45
  twistMotion=Free
 }
 {bone1Name=dyn_mustache_T3_01_r, bone2Name=head, swing1LimitDegrees=10, swing2LimitDegrees=10, twistLimitDegrees=1}
 {bone1Name=dyn_mustache_T3_02_r, bone2Name=dyn_mustache_T3_01_r, swing1LimitDegrees=10, swing2LimitDegrees=10, twistLimitDegrees=1}
 {bone1Name=dyn_mustache_T3_03_r, bone2Name=dyn_mustache_T3_02_r, swing1LimitDegrees=10, swing2LimitDegrees=10, twistLimitDegrees=1}
 {bone1Name=dyn_mustache_T3_01_l, bone2Name=head, swing1LimitDegrees=10, swing2LimitDegrees=10, twistLimitDegrees=1}
 {bone1Name=dyn_mustache_T3_02_l, bone2Name=dyn_mustache_T3_01_l, swing1LimitDegrees=10, swing2LimitDegrees=10, twistLimitDegrees=1}
 {bone1Name=dyn_mustache_T3_03_l, bone2Name=dyn_mustache_T3_02_l, swing1LimitDegrees=10, swing2LimitDegrees=10, twistLimitDegrees=1}
]
