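# Post-copy AnimBP for the tier 3 Pirate face accessory, animating the beard, braids and mustache from the parent mesh's pose.
# Event graph maps head-to-spine_01 rotation from ref pose into alphas for additive beard and braid corrections.
# A head-driven Flap node and clavicle rotation constraints feed a spine_05-space RigidBody simulation, then mustache bones blend toward braid rotation.
# Unconnected AnimDynamics, LookAt, Constraint and extra Flap chains remain outside the output path.
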AnimBlueprint /BRCosmetics/Characters/Player/Male/Medium/Bodies/M_MED_Pirate_Progressive/Meshes/Parts/M_MED_Pirate_T3_FaceAcc_AnimBP
target_skeleton=&/M_MED_Pirate_T3_FaceAcc_Skeleton
graphs [
 AnimGraph {nodes [
   n1 = Anim Dynamics - Bone: dyn_beard_01(Alpha=1, ExternalForce="0,0,0") {
    bGravityOverrideInSimSpace=true
    gravityScale=0.35
    alphaBoolBlend {blendOption=Linear}
    boundBone {boneName=dyn_beard_01}
    physicsBodyDefinitions [{
      sphereCollisionRadius=0
      boundBone {boneName=dyn_beard_01}
      boxExtents=<5 20 10>
      constraintSetup {angularLimitsMax=<0 0 0>, angularLimitsMin=<0 -30 0>, angularTarget=<0 0 0>}
      localJointOffset=<5 0 0>
     }]
    planarLimits [{drivingBone {boneName=spine_05}, planeTransform {location=<0 13 0>, rotation=<rot 0 -20 90>}}]
   }
   n2 = Anim Dynamics - Bone: dyn_beard_02(ComponentPose=@n1, Alpha=1, ExternalForce="0,0,0") {
    bGravityOverrideInSimSpace=true
    gravityScale=0.35
    alphaBoolBlend {blendOption=Linear}
    boundBone {boneName=dyn_beard_02}
    physicsBodyDefinitions [{
      sphereCollisionRadius=0
      boundBone {boneName=dyn_beard_02}
      boxExtents=<5 20 10>
      constraintSetup {angularLimitsMax=<0 0 10>, angularLimitsMin=<0 -30 -10>, angularTarget=<0 0 0>}
      localJointOffset=<5 0 0>
     }]
    planarLimits [{drivingBone {boneName=spine_05}, planeTransform {location=<0 13 0>, rotation=<rot 0 -20 90>}}]
   }
   Anim Dynamics - Bone: dyn_beard_03(ComponentPose=@n2, Alpha=1, ExternalForce="0,0,0") {
    bGravityOverrideInSimSpace=true
    bUseSphericalLimits=true
    gravityScale=0.35
    alphaBoolBlend {blendOption=Linear}
    boundBone {boneName=dyn_beard_03}
    physicsBodyDefinitions [{
      collisionType=CustomSphere
      sphereCollisionRadius=1
      boundBone {boneName=dyn_beard_03}
      boxExtents=<5 20 10>
      constraintSetup {angularLimitsMax=<0 0 15>, angularLimitsMin=<0 -30 -15>, angularTarget=<0 0 0>}
      localJointOffset=<5 0 0>
     }]
    planarLimits [{drivingBone {boneName=spine_05}, planeTransform {location=<0 13 0>, rotation=<rot 0 -20 89.9999>}}]
   }
   Constraint!(ComponentPose, Alpha=1) -> (Pose) {alphaBoolBlend {blendOption=Linear}, boneToModify=~}
   "Constraint - dyn_beard_01 (PP)"!(ComponentPose, Alpha=1, ConstraintWeights_0=1, ConstraintWeights_1:float=1) -> (Pose) {
    alphaBoolBlend {blendOption=Linear}
    boneToModify {boneName=dyn_beard_01}
    constraintSetup [{targetBone {boneName=clavicle_l}}, {targetBone {boneName=clavicle_r}}]
    constraintWeights [1,1]
   }
   n3 = Look At - Bone: dyn_beard_braid_B_T3_master_r(Alpha=1) {
    bUseLookUpAxis=true
    alphaBoolBlend {blendOption=Linear}
    boneToModify {boneName=dyn_beard_braid_B_T3_master_r}
    lookAtLocation {x=0}
    lookAtTarget {bUseSocket=true, socketReference {socketName=beard_B_aim_r}, boneReference=~}
    lookAt_Axis {axis=<-1 0>}
    lookUp_Axis {axis=<0 1>}
   }
   n4 = Look At - Bone: dyn_beard_braid_A_T3_master_r(ComponentPose=@n3, Alpha=1) {
    bUseLookUpAxis=true
    alphaBoolBlend {blendOption=Linear}
    boneToModify {boneName=dyn_beard_braid_A_T3_master_r}
    lookAtLocation {x=0}
    lookAtTarget {bUseSocket=true, socketReference {socketName=beard_A_aim_r}, boneReference=~}
    lookAt_Axis {axis=<-1 0>}
    lookUp_Axis {axis=<0 1>}
   }
   n5 = Look At - Bone: dyn_beard_braid_B_T3_master_l(ComponentPose=@n4, Alpha=1) {
    bUseLookUpAxis=true
    alphaBoolBlend {blendOption=Linear}
    boneToModify {boneName=dyn_beard_braid_B_T3_master_l}
    lookAtLocation {x=0}
    lookAtTarget {bUseSocket=true, socketReference {socketName=beard_B_aim_l}, boneReference=~}
    lookAt_Axis {axis=<1 0>}
    lookUp_Axis {axis=<0 1>}
   }
   Look At - Bone: dyn_beard_braid_A_T3_master_l(ComponentPose=@n5, Alpha=1) {
    bUseLookUpAxis=true
    alphaBoolBlend {blendOption=Linear}
    boneToModify {boneName=dyn_beard_braid_A_T3_master_l}
    lookAtLocation {x=0}
    lookAtTarget {bUseSocket=true, socketReference {socketName=beard_A_aim_l}, boneReference=~}
    lookAt_Axis {axis=<1 0>}
    lookUp_Axis {axis=<0 1>}
   }
   n6 = Flap Node - Bone: dyn_beard_01!(
    ComponentPose:Component Space Pose Link Structure=Local To Component(LocalPose=Copy Pose From Mesh() {bUseAttachedParent=true})
    Alpha:float=1
   ) -> (Pose:Component Space Pose Link Structure) {
    is=title FortAnimGraphNode_Flap
    bUseClamp=true
    targetClampMax=3.14
    targetClampMin=-1
    alphaBoolBlend {blendOption=Linear}
    sourceDrives [{
      bUseRange=true
      drivingCurve=null
      multiplier=-1
      rangeMax=1
      rangeMin=-1
      remappedMax=2
      remappedMin=-2
      sourceComponent=RotationZ
      useQuaternion=true
      sourceBone {boneName=head}
     }]
    targetBone {boneName=dyn_beard_01}
   }
   n7 = "Transform (Modify) Bone - Bone: dyn_beard_01"(ComponentPose=@n6, Alpha=Get head_x_negative() -> (head_x_negative:double), Translation="0,0,0", Rotation="0,40,0", Scale="1,1,1") {rotationMode=BMM_Additive, alphaBoolBlend {blendOption=Linear}, boneToModify {boneName=dyn_beard_01}, rotation {yaw=40}}
   n8 = "Transform (Modify) Bone - Bone: dyn_beard_01"(ComponentPose=@n7, Alpha=Get head_x_positive() -> (head_x_positive:double), Translation="0,0,0", Rotation="0,-40,0", Scale="1,1,1") {rotationMode=BMM_Additive, alphaBoolBlend {blendOption=Linear}, boneToModify {boneName=dyn_beard_01}, rotation {yaw=-40}}
   n9 = "Transform (Modify) Bone - Bone: dyn_beard_01"(
    ComponentPose=@n8
    Alpha=Get head_angle_inverse_alpha() -> (head_angle_inverse_alpha:double)
    Translation="0,0,0"
    Rotation="0,0,-20"
    Scale="1,1,1"
   ) {rotationMode=BMM_Additive, alphaBoolBlend {blendOption=Linear}, boneToModify {boneName=dyn_beard_01}, rotation {roll=-20}}
   n10 = "Constraint - dyn_beard_braid_B_T3_master_l (R)"(ComponentPose=@n9, Alpha=1, ConstraintWeights_0=1) {
    alphaBoolBlend {blendOption=Linear}
    boneToModify {boneName=dyn_beard_braid_B_T3_master_l}
    constraintSetup [{transformType=Rotation, targetBone {boneName=clavicle_l}}]
    constraintWeights [1]
   }
   n11 = "Constraint - dyn_beard_braid_A_T3_master_l (R)"(ComponentPose=@n10, Alpha=1, ConstraintWeights_0=1) {
    alphaBoolBlend {blendOption=Linear}
    boneToModify {boneName=dyn_beard_braid_A_T3_master_l}
    constraintSetup [{transformType=Rotation, targetBone {boneName=clavicle_l}}]
    constraintWeights [1]
   }
   n12 = "Constraint - dyn_beard_braid_B_T3_master_r (R)"(ComponentPose=@n11, Alpha=1, ConstraintWeights_0=1) {
    alphaBoolBlend {blendOption=Linear}
    boneToModify {boneName=dyn_beard_braid_B_T3_master_r}
    constraintSetup [{transformType=Rotation, targetBone {boneName=clavicle_r}}]
    constraintWeights [1]
   }
   n13 = "Constraint - dyn_beard_braid_A_T3_master_r (R)"(ComponentPose=@n12, Alpha=1, ConstraintWeights_0=1) {
    alphaBoolBlend {blendOption=Linear}
    boneToModify {boneName=dyn_beard_braid_A_T3_master_r}
    constraintSetup [{transformType=Rotation, targetBone {boneName=clavicle_r}}]
    constraintWeights [1]
   }
   n14 = "Transform (Modify) Bone - Bone: dyn_beard_braid_A_T3_master_r"(ComponentPose=@n13, Alpha=Get head_x_negative(), Translation="0,0,0", Rotation="0,-30,0", Scale="1,1,1") {rotationMode=BMM_Additive, alphaBoolBlend {blendOption=Linear}, boneToModify {boneName=dyn_beard_braid_A_T3_master_r}, rotation {yaw=-30}}
   n15 = "Transform (Modify) Bone - Bone: dyn_beard_braid_A_T3_master_l"(ComponentPose=@n14, Alpha=Get head_x_positive(), Translation="0,0,0", Rotation="0,30,0", Scale="1,1,1") {rotationMode=BMM_Additive, alphaBoolBlend {blendOption=Linear}, boneToModify {boneName=dyn_beard_braid_A_T3_master_l}, rotation {yaw=30}}
   n16 = Get head_angle_inverse_alpha()
   n17 = "Transform (Modify) Bone - Bone: dyn_beard_braid_A_T3_master_r"(ComponentPose=@n15, Alpha=@n16, Translation="0,0,0", Rotation="0,-30,0", Scale="1,1,1") {rotationMode=BMM_Additive, alphaBoolBlend {blendOption=Linear}, boneToModify {boneName=dyn_beard_braid_A_T3_master_r}, rotation {yaw=-30}}
   n18 = "Transform (Modify) Bone - Bone: dyn_beard_braid_A_T3_master_l"(ComponentPose=@n17, Alpha=@n16, Translation="0,0,0", Rotation="0,40,0", Scale="1,1,1") {rotationMode=BMM_Additive, alphaBoolBlend {blendOption=Linear}, boneToModify {boneName=dyn_beard_braid_A_T3_master_l}, rotation {yaw=40}}
   n19 = "Transform (Modify) Bone - Bone: dyn_beard_braid_B_T3_master_l"(ComponentPose=@n18, Alpha=@n16, Translation="0,0,0", Rotation="0,40,0", Scale="1,1,1") {rotationMode=BMM_Additive, alphaBoolBlend {blendOption=Linear}, boneToModify {boneName=dyn_beard_braid_B_T3_master_l}, rotation {yaw=40}}
   n20 = RigidBody(ComponentPose=@n19) {
    bOverrideWorldGravity=true
    overridePhysicsAsset=&/M_MED_Pirate_T3_FaceAcc_Physics
    simulationSpace=BaseBoneSpace
    alphaBoolBlend {blendOption=Linear}
    baseBoneRef {boneName=spine_05}
    overrideWorldGravity {z=-500}
   }
   n21 = "Constraint - dyn_mustache_T3_01_l (R)"(ComponentPose=@n20, Alpha=@n16, ConstraintWeights_0=1) {
    alphaBoolBlend {blendOption=Linear}
    boneToModify {boneName=dyn_mustache_T3_01_l}
    constraintSetup [{transformType=Rotation, targetBone {boneName=dyn_beard_braid_A_T3_master_l}}]
    constraintWeights [1]
   }
   n22 = "Constraint - dyn_mustache_T3_01_r (R)"(ComponentPose=@n21, Alpha=@n16, ConstraintWeights_0=1) {
    alphaBoolBlend {blendOption=Linear}
    boneToModify {boneName=dyn_mustache_T3_01_r}
    constraintSetup [{transformType=Rotation, targetBone {boneName=dyn_beard_braid_A_T3_master_r}}]
    constraintWeights [1]
   }
   Output Pose(Result=Component To Local(ComponentPose=@n22)) {layerGroup=None}
   n23 = Flap Node - Bone: dyn_beard_braid_B_T3_master_l!(ComponentPose:Component Space Pose Link Structure, Alpha:float=1) -> (Pose:Component Space Pose Link Structure) {
    is=title FortAnimGraphNode_Flap
    bUseClamp=true
    targetClampMax=3.14
    targetClampMin=-3.14
    targetComponent=RotationY
    alphaBoolBlend {blendOption=Linear}
    sourceDrives [{
      bUseRange=false
      drivingCurve=null
      multiplier=2
      rangeMax=0
      rangeMin=0
      remappedMax=0
      remappedMin=0
      sourceComponent=RotationZ
      useQuaternion=true
      sourceBone {boneName=clavicle_l}
     }]
    targetBone {boneName=dyn_beard_braid_B_T3_master_l}
   }
   n24 = Flap Node - Bone: dyn_beard_braid_A_T3_master_l!(ComponentPose:Component Space Pose Link Structure=@n23, Alpha:float=1) -> (Pose:Component Space Pose Link Structure) {
    is=title FortAnimGraphNode_Flap
    bUseClamp=true
    targetClampMax=3.14
    targetClampMin=-3.14
    targetComponent=RotationY
    alphaBoolBlend {blendOption=Linear}
    sourceDrives [{
      bUseRange=false
      drivingCurve=null
      multiplier=2
      rangeMax=0
      rangeMin=0
      remappedMax=0
      remappedMin=0
      sourceComponent=RotationZ
      useQuaternion=true
      sourceBone {boneName=clavicle_l}
     }]
    targetBone {boneName=dyn_beard_braid_A_T3_master_l}
   }
   n25 = Flap Node - Bone: dyn_beard_braid_B_T3_master_r!(ComponentPose:Component Space Pose Link Structure=@n24, Alpha:float=1) -> (Pose:Component Space Pose Link Structure) {
    is=title FortAnimGraphNode_Flap
    bUseClamp=true
    targetClampMax=3.14
    targetClampMin=-3.14
    targetComponent=RotationY
    alphaBoolBlend {blendOption=Linear}
    sourceDrives [{
      bUseRange=false
      drivingCurve=null
      multiplier=2
      rangeMax=0
      rangeMin=0
      remappedMax=0
      remappedMin=0
      sourceComponent=RotationZ
      useQuaternion=true
      sourceBone {boneName=clavicle_r}
     }]
    targetBone {boneName=dyn_beard_braid_B_T3_master_r}
   }
   Flap Node - Bone: dyn_beard_braid_A_T3_master_r!(ComponentPose:Component Space Pose Link Structure=@n25, Alpha:float=1) -> (Pose:Component Space Pose Link Structure) {
    is=title FortAnimGraphNode_Flap
    bUseClamp=true
    targetClampMax=3.14
    targetClampMin=-3.14
    targetComponent=RotationY
    alphaBoolBlend {blendOption=Linear}
    sourceDrives [{
      bUseRange=false
      drivingCurve=null
      multiplier=2
      rangeMax=0
      rangeMin=0
      remappedMax=0
      remappedMin=0
      sourceComponent=RotationZ
      useQuaternion=true
      sourceBone {boneName=clavicle_r}
     }]
    targetBone {boneName=dyn_beard_braid_A_T3_master_r}
   }
   Flap Node - Bone: dyn_beard_01(Alpha=1) {
    alphaBoolBlend {blendOption=Linear}
    sourceDrives [{
      bUseRange=false
      drivingCurve=null
      multiplier=1
      rangeMax=0
      rangeMin=0
      remappedMax=0
      remappedMin=0
      sourceComponent=RotationZ
      useQuaternion=true
      sourceBone {boneName=head}
     }]
    targetBone {boneName=dyn_beard_01}
   }
   Flap Node - Bone: dyn_beard_braid_B_T3_master_r(Alpha=1) {
    bUseClamp=true
    targetClampMax=3.14
    targetClampMin=-3.14
    alphaBoolBlend {blendOption=Linear}
    sourceDrives [{
      bUseRange=false
      drivingCurve=null
      multiplier=-2
      rangeMax=0
      rangeMin=0
      remappedMax=0
      remappedMin=0
      sourceComponent=RotationX
      useQuaternion=true
      sourceBone {boneName=head}
     }]
    targetBone {boneName=dyn_beard_braid_B_T3_master_r}
   }
   Get head_angle_inverse_alpha()
  ]}
 EventGraph {nodes [
   Event BlueprintUpdateAnimation()
   n1 = GetOwnerSkeletalMeshComponent()
   n2 = BreakRotator(InRot=BreakTransform(InTransform=GetDeltaTransformFromRefPose(self=@n1, BoneName=spine_01, BaseName=head)).Rotation)
   > Set head_angle_inverse_alpha(head_angle_inverse_alpha:double=float - float(B=MapRangeClamped(Value=@n2.Yaw, InRangeA=-70, InRangeB=70)) {is=title K2Node_CallFunction}) -> (Output_Get:double)
   > Set head_angle_alpha(head_angle_alpha:double=MapRangeClamped(Value=@n2.Yaw, InRangeA=-70, InRangeB=70)) -> (Output_Get:double)
   > Set head_x_negative(head_x_negative:double=float - float(B=MapRangeClamped(Value=@n2, InRangeA=-40, InRangeB=0))) -> (Output_Get:double)
   > Set head_x_positive(head_x_positive:double=MapRangeClamped(Value=@n2, InRangeA=0, InRangeB=40)) -> (Output_Get:double)
  ]}
]
variables [
 head_angle_alpha {type=number}
 head_angle_inverse_alpha {type=number}
 head_x_negative {type=number}
 head_x_positive {type=number}
]
